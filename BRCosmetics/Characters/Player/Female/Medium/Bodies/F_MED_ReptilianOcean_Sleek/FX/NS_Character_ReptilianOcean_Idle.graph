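# Idle eye-glow effect for the female ReptilianOcean Sleek character body.
# Two looping sprites follow skeletal mesh sockets by spawn group, feeding two sprite renderers, with a visibility tag driven by the head-hidden flag.
# System timeline and time-of-day colors drive tint and fade.
# A blink fade compares two sampled bone orientations by dot product, but its lerps have zero alpha, so it contributes nothing.
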NiagaraSystem /BRCosmetics/Characters/Player/Female/Medium/Bodies/F_MED_ReptilianOcean_Sleek/FX/NS_Character_ReptilianOcean_Idle
dependencies {
 usedDataInterfaces [&NiagaraDataInterfaceCurve]
 usedDynamicInputs []
 usedModules [&SystemState, &Timeline]
}
emitters {eyes_glow_blinkfade {
  emitterUpdateScript [
   &EmitterState(
    Inactive Response=::NewEnumerator1
    Life Cycle Mode=::NewEnumerator1
    Loop Behavior=::NewEnumerator1
    Loop Duration Mode=::NewEnumerator1
    UseLoopDelay=false
   )
   &SpawnBurst_Instantaneous(Spawn Count=2, Spawn Probability=HlslExpression() {hlslExpression=1.0}, Use Spawn Probability=false, UseLoopCountLimit=true)
  ]
  particleSpawnScript [
   &FN_SetSpawnGroup(Number of Spawn Groups=2)
   &InitializeParticle(
    Color=&MultiplyLinearColor(Linear Color=<rgba 400 0 79.9999 1>, Scale Factor=<0.01 0.02 0.02 1>)
    Sprite Size Mode=::NewEnumerator3
    Uniform Sprite Size=1
    UsePositionOffset=false
    Write Material Random=true
   )
   &DynamicMaterialParameters(
    Index 0 Param 1=0.5
    Module.Index0Param0Write=true
    Module.Index0Param1Write=true
    Module.Index0Param2Write=true
    Module.Index0Param3Write=true
   )
   /SetVariables(Particles.CameraOffset=0.7, Particles.CameraOffsetDark=1, Particles.DY0:Vector4f=<1.5 0 0 0>) {isSetParameters=true}
  ]
  particleUpdateScript [
   &ParticleState(
    Kill Particles When Lifetime Has Elapsed=false
    Let Infinitely Lived Particles Die When Emitter Deactivates=true
    Loop Particles Lifetime=true
   )
   &SampleSkeletalMesh(
    AdditionalYawPitchRoll=false
    Bone / Socket Index=2
    Bone Sampling Mode:ENiagaraBoneSamplingMode=::NewEnumerator4
    Position Offset Coordinate Space:ENiagaraCoordinateSpace=::Local
    Skeletal Mesh:NiagaraDataInterfaceSkeletalMesh=$User.SkeletalMesh
   )
   &SampleSkeletalMesh(
    AdditionalYawPitchRoll=false
    Bone / Socket Index=3
    Bone Sampling Mode:ENiagaraBoneSamplingMode=::NewEnumerator4
    Position Offset Coordinate Space:ENiagaraCoordinateSpace=::Local
    Skeletal Mesh:NiagaraDataInterfaceSkeletalMesh=$User.SkeletalMesh
   ) {module=SampleSkeletalMesh001}
   &SkeletalMeshLocation(
    Bone / Socket Index=$Particles.SpawnGroup
    MaskSpawnGroup=false
    Mesh Sampling Type=::NewEnumerator5
    Sampled Position Offset=<0 0 -0.5>
    Skeletal Mesh=$User.SkeletalMesh
    Socket Sampling Mode=::NewEnumerator5
   )
   &ScaleColor(
    Scale Alpha=&Lerp_Float(
     A=&FloatFromCurve(
      FloatCurve=DataInterface() {propertyValues {Curve {keys [{tangentMode=RCTM_User, time=0, value=0}, {tangentMode=RCTM_User, time=1, value=1}]}}}
      CurveIndex=$System.Timeline.Time
     )
     B=&Multiply_Float(
      A=$System.Timeline.Time
      B=&RemapRange(
       Input Value To Remap=&DotProduct(
        Vector A=&ReturnMeshOrientationAxis(Mesh Orientation Axis=::YAxis, Mesh Orientation Quat=$Particles.SampleSkeletalMesh.SampledRotation)
        Vector B=&ReturnMeshOrientationAxis(Mesh Orientation Axis=::YAxis, Mesh Orientation Quat=$Particles.SampleSkeletalMesh001.SampledRotation)
       )
       Input Min=0.6
       Input Max=0.8
       Output Min=1
       Output Max=0
      )
     )
     Alpha=0
    )
    ScaleA=true
    ScaleRGB=true
   )
   &FN_Sprite_Attributes(
    Color=&MultiplyLinearColor(
     Linear Color=&FNTimeOfDayColor(
      Night Color=<rgba 0.4 0 0.0775 0.5>
      Morning Color=<rgba 0.4 0 0.0775 0.5>
      Day Color=<rgba 0.4 0 0.0775 0.5>
      Evening Color=<rgba 0.4 0 0.0775 0.3>
     )
     Scale Factor=&MakeVector4(
      X=1
      Y=1
      Z=1
      W=&Multiply_Float(
       A=$System.Timeline.Time
       B=&Lerp_Float(
        A=&Lerp_Float(A=0, B=0.5, Alpha=&MakeFloatFromInt(Int=&ReturnExecIndex()))
        B=&RemapRange(
         Input Value To Remap=&DotProduct(
          Vector A=&ReturnMeshOrientationAxis(Mesh Orientation Axis=::YAxis, Mesh Orientation Quat=$Particles.SampleSkeletalMesh.SampledRotation)
          Vector B=&ReturnMeshOrientationAxis(Mesh Orientation Axis=::YAxis, Mesh Orientation Quat=$Particles.SampleSkeletalMesh001.SampledRotation)
         )
         Input Min=0.6
         Input Max=0.8
         Output Min=1
         Output Max=0
        )
        Alpha=0
       )
      )
     )
    )
    Create Color Attribute=true
    Create Position Attribute=true
    Create SpriteSize Attribute=true
    Position:NiagaraPosition=&AddVectorToPosition(Position=$Particles.Position, Vector=&TransformVector(Vector=<-0.4 0 0>, Destination Space=::World))
    SpriteSize=&Vector2DFromFloat(Value=5)
    Write to Module Namespace=true
   )
   &ScaleColor(
    Scale Mode=::NewEnumerator1
    Scale RGBA=&MakeVector4_FromLinearColor(LinearColor=&FNTimeOfDayColor(Night Color=<rgba 2 2 2 10>, Morning Color=<rgba 1 1 1 10>, Day Color=<rgba 3 3 3 3>, Evening Color=<rgba 1 1 1 4>))
    ScaleRGBA=true
   ) {module=ScaleColor002}
   /SetVariables(Particles.VisibilityTag=&MakeCustomIntFromBool(Bool=$User.bisHeadHidden, True INT=1, False INT=0)) {isSetParameters=true}
  ]
  rendererClasses [&NiagaraSpriteRendererProperties]
  renderers [
   {rendererClass=&NiagaraSpriteRendererProperties}
   {rendererClass=&NiagaraSpriteRendererProperties, rendererIndex=1}
  ]
 }}
summary {
 systemName=NS_Character_ReptilianOcean_Idle
 emitters [{emitterName=eyes_glow_blinkfade, rendererClasses [&NiagaraSpriteRendererProperties]}]
 userVariables [
  User.ADS {type=NiagaraBool}
  User.bIsFrontEnd {type=NiagaraBool}
  User.bIsFrontEndPreview {type=NiagaraBool}
  User.bisHeadHidden {type=NiagaraBool}
  User.Is_it_in_FrontEnd {type=NiagaraBool}
  User.SkeletalMesh {
   type=NiagaraDataInterfaceSkeletalMesh
   defaultValue {
    struct=NiagaraExt_VariableValue_DataInterface
    value {dataInterface=&self:NiagaraDataInterfaceSkeletalMesh_0}
   }
  }
 ]
}
