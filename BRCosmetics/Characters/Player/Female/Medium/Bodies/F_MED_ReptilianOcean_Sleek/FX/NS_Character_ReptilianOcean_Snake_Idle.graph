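# Niagara idle glow for the snake on the Reptilian Ocean Sleek female character, drawn as sprites by two renderers.
# A single burst of 28 particles tracks the user skeletal mesh by execution index, with per-index camera offsets.
# A system Timeline fades alpha in and drives size during rewind, while variant and time-of-day selectors set both color sets.
# The head-hidden user flag sets each particle's visibility tag.
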NiagaraSystem /BRCosmetics/Characters/Player/Female/Medium/Bodies/F_MED_ReptilianOcean_Sleek/FX/NS_Character_ReptilianOcean_Snake_Idle
dependencies {
 usedDataInterfaces [&NiagaraDataInterfaceCurve, &NiagaraDataInterfaceVector2DCurve, &NiagaraDataInterfaceArrayFloat]
 usedDynamicInputs []
 usedModules [
  &SystemState
  &FN_VariantManager
  &Timeline
 ]
}
emitters {Snake_Glow_Default_Skin {
  emitterUpdateScript [
   &EmitterState(
    Inactive Response=::NewEnumerator1
    Life Cycle Mode=::NewEnumerator1
    Loop Behavior=::NewEnumerator1
    Loop Duration Mode=::NewEnumerator1
    UseLoopDelay=false
   )
   &SpawnBurst_Instantaneous(Spawn Count=28, Spawn Probability=HlslExpression() {hlslExpression=1.0}, Use Spawn Probability=false, UseLoopCountLimit=true)
  ]
  particleSpawnScript [
   &InitializeParticle(Sprite Size Mode=::NewEnumerator3, Uniform Sprite Size=1.5, UsePositionOffset=false, Write Material Random=true)
   &DynamicMaterialParameters(
    Index 0 Param 2=1
    Module.Index0Param0Write=true
    Module.Index0Param1Write=true
    Module.Index0Param2Write=true
    Module.Index0Param3Write=true
   )
  ]
  particleUpdateScript [
   &ParticleState(Kill Particles When Lifetime Has Elapsed=false, Let Infinitely Lived Particles Die When Emitter Deactivates=true)
   &ScaleColor(
    Scale Alpha=&FloatFromCurve(FloatCurve=DataInterface() {propertyValues {Curve {keys [<at 0.05 0>,<at 0.8 0.8>]}}}, CurveIndex=$System.Timeline.Time)
    Scale RGB=&MakeVectorFromLinearColorRGB(Color=&VariantColorSelector(Variant Amount=0, Original Color=<rgba 0 0.0509 0.1 1>, Variant 1 Color=<rgba 0 0.15 0.05 1>, Variant 2 Color=?))
    ScaleA=true
    ScaleRGB=true
   )
   &ScaleSpriteSize(
    Scale Factor=&Vector2DFromCurve(
     Vector2Curve=DataInterface() {propertyValues {
       XCurve {keys [
         {arriveTangent=-11.1433, interpMode=RCIM_Cubic, leaveTangent=-11.1433, tangentMode=RCTM_User, time=0, value=4}
         {arriveTangent=-2.3083, interpMode=RCIM_Cubic, leaveTangent=-2.3083, tangentMode=RCTM_User, time=0.6, value=1}
        ]}
       YCurve {keys [{tangentMode=RCTM_User, time=0, value=0.2}, {tangentMode=RCTM_User, time=0.6, value=1}]}
      }}
     CurveIndex=$System.Timeline.Time
     ScaleCurve=<1 1>
    )
    Scale Sprite Size Mode=::NewEnumerator2
   ) {enabled=false}
   &ScaleSpriteSize(
    Scale Sprite Size Mode=::NewEnumerator0
    Uniform Scale Factor=&Lerp_Float(
     A=1.54
     B=&FloatFromCurve(FloatCurve=DataInterface() {propertyValues {Curve {keys [<at 0 0>,<at 1 1>]}}}, CurveIndex=$System.Timeline.Time)
     Alpha=&MakeCustomFloatFromBool(Bool=$System.Timeline.IsRewinding, True Float=1, False Float=0)
    )
   ) {module=ScaleSpriteSize002}
   &ScaleSpriteSize(
    Scale Sprite Size Mode=::NewEnumerator0
    Uniform Scale Factor=&FloatFromCurve(FloatCurve=DataInterface() {propertyValues {Curve {keys [<at 0 4>,<at 1 1>]}}}, CurveIndex=$System.Timeline.Time)
   ) {enabled=false, module=ScaleSpriteSize001}
   &SkeletalMeshLocation(
    Bone / Socket Index=&ReturnExecIndex()
    MaskSpawnGroup=false
    Mesh Sampling Type=::NewEnumerator5
    Skeletal Mesh=$User.SkeletalMesh
    Socket Sampling Mode=::NewEnumerator5
   )
   /SetVariables(Particles.CameraOffsetDark:NiagaraFloat=&SelectFloatFromArray(
     Direct Array Index=&ReturnExecIndex()
     Float Selection Array=DataInterface() {propertyValues {FloatData [1.2,0.5,1.2,0.5]}}
    )) {isSetParameters=true}
   &CameraOffset(Camera Offset Amount=&SelectFloatFromArray(
     Direct Array Index=&ReturnExecIndex()
     Float Selection Array=DataInterface() {propertyValues {FloatData [0.5,0.5,0.7,0.5]}}
    ))
   &ScaleColor(
    Scale Mode=::NewEnumerator1
    Scale RGBA=&MakeVector4_FromLinearColor(LinearColor=&VariantColorSelector(
      Variant Amount=0
      Original Color=&FNTimeOfDayColor(
       Night Color=<rgba 1.0623 1.5529 5 0.5>
       Morning Color=<rgba 0.1253 0.1005 1 10>
       Day Color=<rgba 0.225 0.2585 0.5 10>
       Evening Color=<rgba 0.2284 0.2396 0.5 10>
      )
      Variant 1 Color=&FNTimeOfDayColor(
       Night Color=<rgba 0.2125 0.3106 1 0.4>
       Morning Color=<rgba 0.1253 0.1005 1 10>
       Day Color=<rgba 0.225 0.2585 0.5 10>
       Evening Color=<rgba 0.2284 0.2396 0.5 10>
      )
      Variant 2 Color=?
     ))
    ScaleRGBA=true
   ) {module=ScaleColor001}
   &FN_Sprite_Attributes(
    Color=&MultiplyLinearColor(
     Linear Color=&MultiplyLinearColor(
      Linear Color=&VariantColorSelector(Variant Amount=0, Original Color=<rgba 0.0718 0.2898 1 0.15>, Variant 1 Color=<rgba 0.05 1 0.7942 0.1>, Variant 2 Color=?)
      Scale Factor=&MakeVector4_FromLinearColor(LinearColor=&VariantColorSelector(
        Variant Amount=0
        Original Color=&FNTimeOfDayColor(
         Night Color=<rgba 1.8208 1.7625 2 0.7>
         Morning Color=<rgba 0.3718 0.5781 1 1>
         Day Color=<rgba 0.8451 0.8909 1 1>
         Evening Color=<rgba 0.3718 0.5781 1 0.5>
        )
        Variant 1 Color=&FNTimeOfDayColor(
         Night Color=<rgba 1.8208 1.7625 2 0.6>
         Morning Color=<rgba 0.3718 0.5781 1 1>
         Day Color=<rgba 0.8451 0.8909 1 1>
         Evening Color=<rgba 0.3718 0.5781 1 0.5>
        )
        Variant 2 Color=?
       ))
     )
     Scale Factor=&MakeVector4(
      X=1
      Y=1
      Z=1
      W=&FloatFromCurve(FloatCurve=DataInterface() {propertyValues {Curve {keys [<at 0.05 0>,<at 0.8 0.8>]}}}, CurveIndex=$System.Timeline.Time)
     )
    )
    Create Color Attribute=true
    Create SpriteSize Attribute=true
    SpriteSize=&Vector2DFromFloat(Value=4.5)
    Write to Module Namespace=true
   )
   /SetVariables(Particles.DY0:Vector4f=&MakeVector4(X=0.7, Y=0, Z=0, W=0)) {isSetParameters=true}
   /SetVariables(Particles.VisibilityTag=&MakeCustomIntFromBool(Bool=$User.bisHeadHidden, True INT=1, False INT=0)) {isSetParameters=true}
  ]
  rendererClasses [&NiagaraSpriteRendererProperties]
  renderers [
   {rendererClass=&NiagaraSpriteRendererProperties}
   {rendererClass=&NiagaraSpriteRendererProperties, rendererIndex=1}
  ]
 }}
summary {
 systemName=NS_Character_ReptilianOcean_Snake_Idle
 emitters [{emitterName=Snake_Glow_Default_Skin, rendererClasses [&NiagaraSpriteRendererProperties]}]
 userVariables [
  User.ADS {type=NiagaraBool, defaultValue=true}
  User.bIsFrontEnd {type=NiagaraBool}
  User.bIsFrontEndPreview {type=NiagaraBool}
  User.bisHeadHidden {type=NiagaraBool}
  User.Is_it_in_FrontEnd {type=NiagaraBool}
  User.SkeletalMesh {
   type=NiagaraDataInterfaceSkeletalMesh
   defaultValue {
    struct=NiagaraExt_VariableValue_DataInterface
    value {dataInterface=&self:NiagaraDataInterfaceSkeletalMesh_0}
   }
  }
  User.Variant {type=NiagaraFloat, defaultValue=0, description=""}
 ]
}
