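# Control Rig for the snake-hair face accessory of the female ReptilianOcean Sleek character, posing and animating snake bones on the head.
# When IsTargeting is true it applies a fixed transform adjustment, and the branch's Completed path always runs afterward.
# That path uses three spherical pose readers, measuring head against spine_05, to weight corrective snake-bone transforms.
# A second sequence step adds wave and pendulum harmonics to two front chains (fl, fr), then applies fixed offsets.
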ControlRigBlueprint /BRCosmetics/Characters/Player/Female/Medium/Bodies/F_MED_ReptilianOcean_Sleek/Meshes/Parts/F_MED_ReptilianOcean_Sleek_FaceAcc_CtrlRig
graphs [
 RigVMFunctionLibrary {links [], nodes []}
 RigVMModel {nodes [
   RigUnit_BeginExecution = &RigUnit_BeginExecution()
   RigVMFunction_Sequence = &RigVMFunction_Sequence(ExecuteContext=@RigUnit_BeginExecution.ExecutePin)
   RigVMFunction_ControlFlowBranch = &RigVMFunction_ControlFlowBranch(ExecuteContext=@RigVMFunction_Sequence.A, Condition=RigVMVariableNode() {variable=IsTargeting}.Value)
   n1 = &RigUnit_SphericalPoseReader(
    ActiveRegionScaleFactors=(PositiveWidth=1,NegativeWidth=3,PositiveHeight=3,NegativeHeight=3)
    ActiveRegionSize=0.1
    Debug=(bDrawDebug=false,bDraw2D=False,bDrawLocalAxes=False,DebugScale=25,DebugSegments=20,DebugThickness=0.25)
    DriverAxis=(X=1,Y=0,Z=0)
    DriverItem=(Type=Bone,Name="head")
    FalloffRegionScaleFactors=(PositiveWidth=1,NegativeWidth=1,PositiveHeight=1,NegativeHeight=1)
    FalloffSize=0.3
    FlipHeightScaling=False
    FlipWidthScaling=False
    OptionalParentItem=(Type=Bone,Name="spine_05")
    RotationOffset=(X=-90,Y=0,Z=20)
    ExecutePin=@RigVMFunction_ControlFlowBranch.Completed
   )
   > &RigUnit_ModifyTransforms(
    ItemToModify="((Item=(Type=Bone,Name=\"dyn_snake_hair_d01\"),Transform=(Rotation=(X=0.0934,Y=0.7744,Z=0.0562,W=0.6232),Translation=(X=0,Y=0,Z=0),Scale3D=(X=1,Y=1,Z=1))),(Item=(Type=Bone,Name=\"dyn_snake_hair_g01\"),Transform=(Rotation=(X=0.0474,Y=0.3282,Z=-0.01,W=0.9434),Tr ...(truncated)"
    WeightMinimum=0
    Weight=@n1.OutputParam
   )
   > n2 = &RigUnit_SphericalPoseReader(
    ActiveRegionScaleFactors=(PositiveWidth=1,NegativeWidth=3,PositiveHeight=3,NegativeHeight=3)
    ActiveRegionSize=0.1
    Debug=(bDrawDebug=false,bDraw2D=false,bDrawLocalAxes=false,DebugScale=25,DebugSegments=20,DebugThickness=0.25)
    DriverAxis=(X=1,Y=0,Z=0)
    DriverItem=(Type=Bone,Name="head")
    FalloffRegionScaleFactors=(PositiveWidth=1,NegativeWidth=1,PositiveHeight=1,NegativeHeight=1)
    FalloffSize=0.5
    FlipHeightScaling=false
    FlipWidthScaling=false
    OptionalParentItem=(Type=Bone,Name="spine_05")
    RotationOffset=(X=90,Y=0,Z=5)
   )
   > &RigUnit_ModifyTransforms(
    ItemToModify=((Item=(Type=Bone,Name="dyn_snake_hair_c01"),Transform=(Rotation=(X=0,Y=0,Z=0.1736,W=0.9848),Translation=(X=0,Y=5,Z=0),Scale3D=(X=1,Y=1,Z=1))))
    WeightMinimum=0
    Weight=@n2.OutputParam
   )
   > n3 = &RigUnit_SphericalPoseReader(
    ActiveRegionScaleFactors=(PositiveWidth=1,NegativeWidth=3,PositiveHeight=3,NegativeHeight=3)
    ActiveRegionSize=0.1
    Debug=(bDrawDebug=false,bDraw2D=False,bDrawLocalAxes=False,DebugScale=25,DebugSegments=20,DebugThickness=0.25)
    DriverAxis=(X=1,Y=0,Z=0)
    DriverItem=(Type=Bone,Name="head")
    FalloffRegionScaleFactors=(PositiveWidth=1,NegativeWidth=1,PositiveHeight=1,NegativeHeight=1)
    FalloffSize=0.5
    FlipHeightScaling=False
    FlipWidthScaling=False
    OptionalParentItem=(Type=Bone,Name="spine_05")
    RotationOffset=(X=-90,Y=0,Z=5)
   )
   > &RigUnit_ModifyTransforms(
    ItemToModify="((Item=(Type=Bone,Name=\"dyn_snake_hair_c01\"),Transform=(Rotation=(X=0,Y=0,Z=0,W=1),Translation=(X=0,Y=10,Z=0),Scale3D=(X=1,Y=1,Z=1))),(Item=(Type=Bone,Name=\"dyn_snake_hair_d02\"),Transform=(Rotation=(X=0,Y=0,Z=0,W=1),Translation=(X=0,Y=0,Z=0),Scale3D=(X=1,Y ...(truncated)"
    WeightMinimum=0
    Weight=@n3.OutputParam
   )
   &RigUnit_ModifyTransforms(
    ItemToModify="((Item=(Type=Bone,Name=\"dyn_snake_hair_c01\"),Transform=(Rotation=(X=-0.2549,Y=-0.0449,Z=0.1677,W=0.9513),Translation=(X=0,Y=1,Z=0),Scale3D=(X=1,Y=1,Z=1))),(Item=(Type=Bone,Name=\"dyn_snake_hair_d01\"),Transform=(Rotation=(X=0,Y=0.1305,Z=0,W=0.9914),Translati ...(truncated)"
    Weight=1
    WeightMinimum=0
    ExecutePin=@RigVMFunction_ControlFlowBranch.True
   )
   /Script/ControlRig.RigUnit_ChainHarmonicsPerItem(
    ChainRoot=(Type=Bone,Name="dyn_snake_hair_fl01")
    DrawWorldOffset=(Rotation=(X=0,Y=0,Z=0,W=1),Translation=(X=0,Y=0,Z=0),Scale3D=(X=1,Y=1,Z=1))
    Pendulum=(bEnabled=true,PendulumStiffness=2,PendulumGravity=(X=0,Y=0,Z=0),PendulumBlend=0.75,PendulumDrag=0.98,PendulumMinimum=0,PendulumMaximum=0.1,PendulumEase=Linear,UnwindAxis=(X=0,Y=1,Z=0),UnwindMinimum=0.2,UnwindMaximum=0.05)
    Reach=(bEnabled=false,ReachTarget=(X=0,Y=0,Z=0),ReachAxis=(X=1,Y=0,Z=0),ReachMinimum=0,ReachMaximum=0,ReachEase=Linear)
    Speed=(X=1,Y=0.8,Z=1.1)
    Wave=(bEnabled=True,WaveFrequency=(X=0,Y=0.8,Z=0.2),WaveAmplitude=(X=0.2,Y=0.5,Z=0.2),WaveOffset=(X=0,Y=0,Z=0.2),WaveNoise=(X=0,Y=0.2,Z=0.2),WaveMinimum=0,WaveMaximum=1,WaveEase=Linear)
    WaveCurve=(EditorCurveData=(Keys=((),(Time=1,Value=1))))
    bDrawDebug=True
    ExecutePin=@RigVMFunction_Sequence.B
   ) {is=struct RigVMUnitNode}
   > &RigUnit_ChainHarmonicsPerItem(
    ChainRoot=(Type=Bone,Name="dyn_snake_hair_fr01")
    DrawWorldOffset=(Rotation=(X=0,Y=0,Z=0,W=1),Translation=(X=0,Y=0,Z=0),Scale3D=(X=1,Y=1,Z=1))
    Pendulum=(bEnabled=true,PendulumStiffness=2,PendulumGravity=(X=0,Y=0,Z=0),PendulumBlend=0.75,PendulumDrag=0.98,PendulumMinimum=0,PendulumMaximum=0.1,PendulumEase=Linear,UnwindAxis=(X=0,Y=1,Z=0),UnwindMinimum=0.2,UnwindMaximum=0.05)
    Reach=(bEnabled=false,ReachTarget=(X=0,Y=0,Z=0),ReachAxis=(X=1,Y=0,Z=0),ReachMinimum=0,ReachMaximum=0,ReachEase=Linear)
    Speed=(X=1,Y=0.8,Z=0.4)
    Wave=(bEnabled=True,WaveFrequency=(X=0,Y=-0.6,Z=0.2),WaveAmplitude=(X=0.5,Y=0,Z=0.6),WaveOffset=(X=0,Y=2,Z=2),WaveNoise=(X=0,Y=0.2,Z=0.2),WaveMinimum=0,WaveMaximum=1,WaveEase=Linear)
    WaveCurve=(EditorCurveData=(Keys=((),(Time=1,Value=1))))
    bDrawDebug=True
   )
   > &RigUnit_ModifyTransforms(
    ItemToModify="((Item=(Type=Bone,Name=\"dyn_snake_hair_fl04\"),Transform=(Rotation=(X=0.0246,Y=0.3679,Z=0.1183,W=0.922),Translation=(X=0,Y=0.3,Z=0),Scale3D=(X=1,Y=1,Z=1))),(Item=(Type=Bone,Name=\"dyn_snake_hair_fr04\"),Transform=(Rotation=(X=0.25,Y=0.25,Z=-0.067,W=0.933),Tra ...(truncated)"
    Weight=1
    WeightMinimum=0
   )
  ]}
]
hierarchy [
 root {loc=<0 0 0>, rot=<0 0 0 1>}
 pelvis {parent=root, loc=<0 -0.5301 98.0571>, rot=<0.0127 0.707 -0.0127 -0.707>}
 spine_01 {parent=pelvis, loc=<6.1208 -0.9328 0>, rot=<0 0 0.034 -0.9994>}
 spine_02 {parent=spine_01, loc=<6.5996 -0.2095 0>, rot=<0 0 0.0238 -0.9997>}
 spine_03 {parent=spine_02, loc=<6.4914 -0.1237 0>, rot=<0 0 0.0149 -0.9999>}
 spine_04 {parent=spine_03, loc=<6.9598 0.0794 0>, rot=<0 0 -0.0132 -0.9999>}
 spine_05 {parent=spine_04, loc=<6.9987 0 0>, rot=<0 0 -0.0015 -1>}
 neck_01 {parent=spine_05, loc=<11.3138 0.1529 0>, rot=<0 0 -0.2003 -0.9797>}
 neck_02 {parent=neck_01, loc=<4.5627 0 0>, rot=<0 0 0.0504 -0.9987>}
 head {parent=neck_02, loc=<5.3914 0 0>, rot=<0 0 0.115 -0.9934>}
 dyn_snake_hair_fl01 {parent=head, loc=<7.4025 8.9528 -4.04>, rot=<-0.1308 -0.5678 0.7403 -0.3353>}
 dyn_snake_hair_fl02 {parent=dyn_snake_hair_fl01, loc=<3.8061 0 0>, rot=<-0.0595 -0.0919 0.5406 -0.8341>}
 dyn_snake_hair_fl03 {parent=dyn_snake_hair_fl02, loc=<3.6021 0 0>, rot=<-0.0421 -0.1625 -0.2471 -0.9543>}
 dyn_snake_hair_fl04 {parent=dyn_snake_hair_fl03, loc=<2.1163 0 0>, rot=<0 0 0.2475 -0.9689>}
 dyn_snake_hair_fr01 {parent=head, loc=<6.6549 3.5812 6.8141>, rot=<-0.1045 -0.9573 0.2487 0.1036>}
 dyn_snake_hair_fr02 {parent=dyn_snake_hair_fr01, loc=<3.1866 0 0>, rot=<0.0699 0.0729 -0.3918 -0.9145>}
 dyn_snake_hair_fr03 {parent=dyn_snake_hair_fr02, loc=<3.7895 0 0>, rot=<0.0787 0.0989 0.6177 -0.7762>}
 dyn_snake_hair_fr04 {parent=dyn_snake_hair_fr03, loc=<2.6983 0 0>, rot=<0 0 -0.4706 -0.8824>}
 dyn_snake_hair_a01 {parent=head, loc=<4.4102 -8.525 1.2232>, rot=<-0.1436 -0.8746 -0.446 0.1249>}
 dyn_snake_hair_a02 {parent=dyn_snake_hair_a01, loc=<6.6867 0 -0.3705>, rot=<0 -0.3926 0 -0.9197>}
 dyn_snake_hair_a03 {parent=dyn_snake_hair_a02, loc=<4.9862 0 -0.144>, rot=<0 -0.1458 0 -0.9893>}
 dyn_snake_hair_a04 {parent=dyn_snake_hair_a03, loc=<5.4341 0 -0.1726>, rot=<0 -0.1241 0 -0.9923>}
 dyn_snake_hair_a_tongue {parent=dyn_snake_hair_a04, loc=<2.6878 0.7162 1.1383>, rot=<-0.106 0.1057 -0.0855 -0.985>}
 dyn_snake_hair_a_tongue_end {parent=dyn_snake_hair_a_tongue, loc=<1.2144 -0.012 0.0744>, rot=<0 0 0 1>}
 dyn_snake_hair_b01 {parent=head, loc=<0.7606 -6.5059 -0.6331>, rot=<0.1164 -0.8711 -0.4434 0.1762>}
 dyn_snake_hair_b02 {parent=dyn_snake_hair_b01, loc=<5.1263 0 -0.0773>, rot=<0 -0.2042 0 -0.9789>}
 dyn_snake_hair_b03 {parent=dyn_snake_hair_b02, loc=<5.0418 0 -0.0948>, rot=<0.0357 -0.2254 -0.03 -0.9731>}
 dyn_snake_hair_b04 {parent=dyn_snake_hair_b03, loc=<4.9627 0 -0.0529>, rot=<0 -0.1089 0 -0.9941>}
 dyn_snake_hair_b_jaw {parent=dyn_snake_hair_b04, loc=<1.8581 -0.3032 0.532>, rot=<0.2116 0.0347 -0.0075 -0.9767>}
 dyn_snake_hair_c01 {parent=head, loc=<-0.3472 -5.0492 -2.1178>, rot=<0.0894 -0.7365 -0.6628 0.1012>}
 dyn_snake_hair_c02 {parent=dyn_snake_hair_c01, loc=<5.1263 0 -0.0773>, rot=<0 -0.2042 0 -0.9789>}
 dyn_snake_hair_c03 {parent=dyn_snake_hair_c02, loc=<5.0418 0 -0.0948>, rot=<0.0357 -0.2254 -0.03 -0.9731>}
 dyn_snake_hair_c04 {parent=dyn_snake_hair_c03, loc=<4.3335 0 0.0866>, rot=<0 -0.1089 0 -0.9941>}
 dyn_snake_hair_c_tongue {parent=dyn_snake_hair_c04, loc=<3.8496 -0.3807 1.5793>, rot=<0.0769 0.1091 0.0437 -0.9901>}
 dyn_snake_hair_c_tongue_end {parent=dyn_snake_hair_c_tongue, loc=<1.2345 0.0229 0.1003>, rot=<0 0 0 1>}
 dyn_snake_hair_d01 {parent=head, loc=<5.0339 -8.1456 -2.2505>, rot=<-0.0885 0.4538 0.8822 0.0893>}
 dyn_snake_hair_d01_a_01 {parent=dyn_snake_hair_d01, loc=<6.2954 1.4909 -0.1638>, rot=<0 -0.1026 0 -0.9947>}
 dyn_snake_hair_d01_a_02 {parent=dyn_snake_hair_d01_a_01, loc=<4.1679 0.1757 -0.4163>, rot=<0.0055 -0.2845 -0.0048 -0.9587>}
 dyn_snake_hair_d01_a_03 {parent=dyn_snake_hair_d01_a_02, loc=<3.8526 -0.1201 -0.2104>, rot=<-0.0133 -0.2165 0.008 -0.9762>}
 dyn_snake_hair_d01_a_04 {parent=dyn_snake_hair_d01_a_03, loc=<3.5694 -0.1384 -0.3516>, rot=<-0.0103 0.025 -0.0002 -0.9996>}
 dyn_snake_hair_d01_a_jaw {parent=dyn_snake_hair_d01_a_04, loc=<1.256 0.2956 0.1839>, rot=<0.111 -0.052 0.0058 -0.9924>}
 dyn_snake_hair_d01_b_01 {parent=dyn_snake_hair_d01, loc=<4.6024 -1.1799 0.1999>, rot=<0.1061 -0.0987 -0.0277 -0.989>}
 dyn_snake_hair_d01_b_02 {parent=dyn_snake_hair_d01_b_01, loc=<4.8786 0 0.1827>, rot=<0 -0.2422 0 -0.9702>}
 dyn_snake_hair_d01_b_03 {parent=dyn_snake_hair_d01_b_02, loc=<3.5413 0 0>, rot=<0 0 0 -1>}
 dyn_snake_hair_d01_b_jaw {parent=dyn_snake_hair_d01_b_03, loc=<0.9603 -0.2268 0.0207>, rot=<0.0342 -0.0438 -0.0015 -0.9985>}
 dyn_snake_hair_d02 {parent=dyn_snake_hair_d01, loc=<6.2954 0.3234 -0.1638>, rot=<0 -0.1026 0 -0.9947>}
 dyn_snake_hair_d03 {parent=dyn_snake_hair_d02, loc=<4.2363 1.3181 0.1198>, rot=<0 -0.1945 0 -0.9809>}
 dyn_snake_hair_d04 {parent=dyn_snake_hair_d03, loc=<3.7519 0 0.2547>, rot=<0 -0.1659 0 -0.9861>}
 dyn_snake_hair_d05 {parent=dyn_snake_hair_d04, loc=<3.7735 0 0>, rot=<0 0 0 1>}
 dyn_snake_hair_d01_a {parent=dyn_snake_hair_d02, loc=<3.2491 -1.603 0.8959>, rot=<0.1082 -0.3058 0.0178 -0.9458>}
 dyn_snake_hair_d02_a {parent=dyn_snake_hair_d01_a, loc=<3.8277 0 0.1827>, rot=<0 0 0 1>}
 dyn_snake_hair_e01 {parent=head, loc=<7.1887 -8.3051 -3.8284>, rot=<-0.2041 0.356 0.9074 0.09>}
 dyn_snake_hair_e02 {parent=dyn_snake_hair_e01, loc=<5.6704 0 -0.0281>, rot=<0 -0.1966 0 -0.9805>}
 dyn_snake_hair_e03 {parent=dyn_snake_hair_e02, loc=<4.4064 0 0.1827>, rot=<0 -0.1945 0 -0.9809>}
 dyn_snake_hair_e04 {parent=dyn_snake_hair_e03, loc=<3.7518 0 0.2547>, rot=<0 -0.0664 0 -0.9978>}
 dyn_snake_hair_e_tongue {parent=dyn_snake_hair_e04, loc=<3.5505 0.5386 0.811>, rot=<0.0239 0.0394 0.0009 -0.9989>}
 dyn_snake_hair_e_tongue_end {parent=dyn_snake_hair_e_tongue, loc=<1.2122 0.0642 0.1133>, rot=<0 0 0 1>}
 dyn_snake_hair_f01 {parent=head, loc=<4.7831 -3.5496 -5.7011>, rot=<-0.0373 0.1935 0.9797 -0.0356>}
 dyn_snake_hair_f02 {parent=dyn_snake_hair_f01, loc=<5.6819 0.0051 -0.0191>, rot=<-0.0018 -0.0998 -0.0015 -0.995>}
 dyn_snake_hair_f03 {parent=dyn_snake_hair_f02, loc=<4.3936 -0.0323 0.0951>, rot=<-0.0062 -0.304 0.0062 -0.9526>}
 dyn_snake_hair_f04 {parent=dyn_snake_hair_f03, loc=<3.7676 -0.0065 0.1425>, rot=<-0.0011 -0.0855 0.0038 -0.9963>}
 dyn_snake_hair_f_jaw {parent=dyn_snake_hair_f04, loc=<1.5618 0.1281 0.1175>, rot=<-0.1181 0.0024 -0.0316 -0.9925>}
 dyn_snake_hair_g01 {parent=head, loc=<6.1881 -2.7194 -10.2904>, rot=<-0.059 0.1486 0.9863 0.0411>}
 dyn_snake_hair_g01_a {parent=dyn_snake_hair_g01, loc=<2.6105 1.4736 1.4543>, rot=<-0.0026 -0.0023 -0.0006 -1>}
 dyn_snake_hair_g02_a {parent=dyn_snake_hair_g01_a, loc=<5.708 -0.0073 -0.0478>, rot=<0.0001 -0.1873 0.0014 -0.9823>}
 dyn_snake_hair_g03_a {parent=dyn_snake_hair_g02_a, loc=<4.4588 -0.0013 0.1852>, rot=<0.0009 -0.2169 0.0007 -0.9762>}
 dyn_snake_hair_g04_a {parent=dyn_snake_hair_g03_a, loc=<3.7838 0.0008 0.2794>, rot=<0.0014 -0.0642 0.0001 -0.9979>}
 dyn_snake_hair_g_a_jaw {parent=dyn_snake_hair_g04_a, loc=<0.5456 0.1727 0.3858>, rot=<0.06 0.0679 0.011 -0.9958>}
 dyn_snake_hair_g01_b {parent=dyn_snake_hair_g01, loc=<2.7401 -1.3792 0.6356>, rot=<0 -0.0539 0 -0.9985>}
 dyn_snake_hair_g02_b {parent=dyn_snake_hair_g01_b, loc=<4.1935 0 0.0476>, rot=<0 -0.2407 0 -0.9706>}
 dyn_snake_hair_g03_b {parent=dyn_snake_hair_g02_b, loc=<4.4064 0 0.1827>, rot=<0 -0.2191 0 -0.9757>}
 dyn_snake_hair_g_b_tongue {parent=dyn_snake_hair_g03_b, loc=<4.5295 -0.2076 1.0307>, rot=<0 0.0577 0 -0.9983>}
 dyn_snake_hair_g_b_tongue_end {parent=dyn_snake_hair_g_b_tongue, loc=<1.2354 -0.0348 0.0793>, rot=<0 0 0 1>}
 dyn_snake_hair_g02 {parent=dyn_snake_hair_g01, loc=<2.6839 1.4693 0.63>, rot=<0 0 0 1>}
 dyn_snake_hair_g03 {parent=dyn_snake_hair_g02, loc=<5.6704 0 -0.0281>, rot=<0 -0.1856 0 -0.9826>}
 dyn_snake_hair_g04 {parent=dyn_snake_hair_g03, loc=<4.4064 0 0.1827>, rot=<0 -0.2191 0 -0.9757>}
 dyn_snake_hair_g05 {parent=dyn_snake_hair_g04, loc=<3.7519 0 0.2547>, rot=<0 -0.0664 0 -0.9978>}
 dyn_snake_hair_h01 {parent=head, loc=<6.1029 2.0898 -7.4651>, rot=<-0.0619 0.0753 0.9946 0.0366>}
 dyn_snake_hair_h02 {parent=dyn_snake_hair_h01, loc=<2.7101 1.2167 0.0543>, rot=<0.0337 0.0137 -0.0255 -0.999>}
 dyn_snake_hair_h03 {parent=dyn_snake_hair_h02, loc=<5.6652 -0.0231 -0.1295>, rot=<0.007 -0.1192 0.0028 -0.9928>}
 dyn_snake_hair_h04 {parent=dyn_snake_hair_h03, loc=<4.5889 -0.0031 -0.1097>, rot=<0.0122 -0.2559 0.0012 -0.9666>}
 dyn_snake_hair_h05 {parent=dyn_snake_hair_h04, loc=<4.8287 0.0242 -0.0848>, rot=<0.0085 -0.0946 -0.0069 -0.9955>}
 dyn_snake_hair_h_tongue {parent=dyn_snake_hair_h05, loc=<4.8051 0.2047 -0.2781>, rot=<0.1689 0 0 -0.9856>}
 dyn_snake_hair_h_tongue_end {parent=dyn_snake_hair_h_tongue, loc=<1.4124 0.0051 0.0095>, rot=<0 0 0 1>}
 dyn_snake_hair_h01_a {parent=dyn_snake_hair_h01, loc=<2.7351 -1.2319 0.541>, rot=<0.1666 -0.0479 0.0223 -0.9846>}
 dyn_snake_hair_h02_a {parent=dyn_snake_hair_h01_a, loc=<4.1935 0 0.0476>, rot=<0 -0.2713 0 -0.9625>}
 dyn_snake_hair_h03_a {parent=dyn_snake_hair_h02_a, loc=<4.4064 0 0.1827>, rot=<0 -0.2191 0 -0.9757>}
 dyn_snake_hair_h_a_jaw {parent=dyn_snake_hair_h03_a, loc=<1.479 0.0779 0.5994>, rot=<0.0597 0.1836 0.0112 -0.9811>}
 dyn_snake_hair_i01 {parent=head, loc=<6.7841 3.3336 -8.3823>, rot=<-0.4293 0.0476 0.9016 0.0216>}
 dyn_snake_hair_i02 {parent=dyn_snake_hair_i01, loc=<4.6052 -0.0185 -0.0229>, rot=<0.0035 -0.1374 0.001 -0.9905>}
 dyn_snake_hair_i03 {parent=dyn_snake_hair_i02, loc=<4.2658 0.0117 0.0214>, rot=<0.0019 -0.01 -0.003 -0.9999>}
 dyn_snake_hair_i_jaw {parent=dyn_snake_hair_i03, loc=<1.371 -0.0145 0.3898>, rot=<0 0 0 1>}
 dyn_snake_hair_j01 {parent=head, loc=<7.9124 6.5259 -8.39>, rot=<-0.2747 -0.1409 0.9508 -0.0273>}
 dyn_snake_hair_j02 {parent=dyn_snake_hair_j01, loc=<4.603 0 0.2339>, rot=<0 -0.1762 0 -0.9844>}
 dyn_snake_hair_j03 {parent=dyn_snake_hair_j02, loc=<4.1358 0 0.1792>, rot=<0 0.0066 0 -1>}
 dyn_snake_hair_j_tongue {parent=dyn_snake_hair_j03, loc=<3.1981 -0.0226 0.1427>, rot=<0.0715 0 0 -0.9974>}
 dyn_snake_hair_j_tongue_end {parent=dyn_snake_hair_j_tongue, loc=<1.4101 0.0196 -0.021>, rot=<0 0 0 1>}
 dyn_snake_hair_k01 {parent=head, loc=<7.5484 9.5176 -7.3808>, rot=<-0.224 -0.2709 0.9362 0.005>}
 dyn_snake_hair_k02 {parent=dyn_snake_hair_k01, loc=<5.2191 0 0.0061>, rot=<0 -0.1599 0 -0.9871>}
 dyn_snake_hair_k03 {parent=dyn_snake_hair_k02, loc=<4.7926 0 0.1878>, rot=<0 -0.0268 0 -0.9996>}
 dyn_snake_hair_k_jaw {parent=dyn_snake_hair_k03, loc=<0.6284 0.3347 0.3161>, rot=<0.111 -0.0013 0.0001 -0.9938>}
 dyn_hair_a {parent=head, loc=<12.5055 4.3837 2.7982>, rot=<0.142 -0.9271 0.1523 -0.3117>}
 dyn_hair_b {parent=head, loc=<9.803 -5.9149 3.0303>, rot=<-0.423 -0.8371 -0.0571 -0.3422>}
 clavicle_l {parent=spine_05, loc=<5.4831 -0.1053 -3.2946>, rot=<0.0465 -0.7289 -0.009 -0.6829>}
 upperarm_l {parent=clavicle_l, loc=<9.8658 -1.2136 -0.4105>, rot=<0.0164 -0.3245 -0.0269 -0.9454>}
 clavicle_r {parent=spine_05, loc=<5.4831 -0.1053 3.2946>, rot=<0.7289 0.0465 -0.6829 0.009>}
 upperarm_r {parent=clavicle_r, loc=<-9.8658 1.2136 0.4105>, rot=<0.0164 -0.3245 -0.0269 -0.9454>}
]
variables [IsTargeting {type=bool, default=False}]
