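# Physics asset for the ReptilianOcean Sleek female character's face accessory mesh, defining physics for its snake-like hair.
# Capsule bodies form multi-bone strand chains rooted at the head, some branching, alongside capsules for head, neck, upper spine, left clavicle and upper arms.
# Most strands have narrow swing and twist limits, tighter on several later chains.
# The fl and fr strands and left-arm links rotate freely.
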PhysicsAsset /BRCosmetics/Characters/Player/Female/Medium/Bodies/F_MED_ReptilianOcean_Sleek/Meshes/Parts/F_MED_ReptilianOcean_Sleek_FaceAcc_Physics
bodies {
 clavicle_l [{
   shapeType=Capsule
   length=12.4425
   radius=7.13
   shapeName=clavicle_l_capsule
   center=<2.4909 -1.3089 -1.2619>
   rotation=<rot -0.6777 -87.4431 104.8309>
  }]
 dyn_snake_hair_a01 [{
   shapeType=Capsule
   length=2.0093
   radius=1.3334
   shapeName=dyn_snake_hair_a01_capsule
   center=<4.199 -0.2357 0.3239>
   rotation=<rot 0 93.4253 -87.8471>
  }]
 dyn_snake_hair_a02 [{
   shapeType=Capsule
   length=1.8194
   radius=1.9825
   shapeName=dyn_snake_hair_a02_capsule
   center=<1.874 -0.3849 0.2702>
   rotation=<rot 0 -113.1502 -86.7522>
  }]
 dyn_snake_hair_a03 [{
   shapeType=Capsule
   length=2.4258
   radius=1.7517
   shapeName=dyn_snake_hair_a03_capsule
   center=<1.3501 -0.035 -0.0784>
   rotation=<rot 0 92.2526 -89.9295>
  }]
 dyn_snake_hair_a04 [{
   shapeType=Capsule
   length=3.0473
   radius=1.9905
   shapeName=dyn_snake_hair_a04_capsule
   center=<1.6495 0.623 0.5698>
   rotation=<rot 0 95.561 -68.4074>
  }]
 dyn_snake_hair_b01 [{
   shapeType=Capsule
   length=1.5504
   radius=1.5271
   shapeName=dyn_snake_hair_b01_capsule
   center=<1.7615 0.3524 0.5273>
   rotation=<rot 0 105.3055 -78.3475>
  }]
 dyn_snake_hair_b02 [{
   shapeType=Capsule
   length=4.1078
   radius=1.605
   shapeName=dyn_snake_hair_b02_capsule
   center=<2.7918 0.0256 0.3053>
   rotation=<rot 0 -95.3883 -75>
  }]
 dyn_snake_hair_b03 [{
   shapeType=Capsule
   length=2.6908
   radius=2.3233
   shapeName=dyn_snake_hair_b03_capsule
   center=<2.5215 -0.0074 0.4105>
   rotation=<rot 3.1213 -96.6968 -71.9857>
  }]
 dyn_snake_hair_b04 [{
   shapeType=Capsule
   length=1.568
   radius=2.6414
   shapeName=dyn_snake_hair_b04_capsule
   center=<3.2714 -0.4236 0.3117>
   rotation=<rot 0 97.6229 -73.7804>
  }]
 dyn_snake_hair_c01 [{
   shapeType=Capsule
   length=3.8414
   radius=1.4229
   shapeName=dyn_snake_hair_c01_capsule
   center=<6.5903 -0.1458 -1.5543>
   rotation=<rot 0 -92.7172 -64.0683>
  }]
 dyn_snake_hair_c02 [{
   shapeType=Capsule
   length=2.2381
   radius=1.3139
   shapeName=dyn_snake_hair_c02_capsule
   center=<3.8164 0.1507 -0.1219>
   rotation=<rot 0 -72.6844 -60.8808>
  }]
 dyn_snake_hair_c03 [{
   shapeType=Capsule
   length=2.3202
   radius=1.3667
   shapeName=dyn_snake_hair_c03_capsule
   center=<2.9651 -0.2373 0.1515>
   rotation=<rot 0 90.7202 -76.8062>
  }]
 dyn_snake_hair_c04 [{
   shapeType=Capsule
   length=3.069
   radius=1.5842
   shapeName=dyn_snake_hair_c04_capsule
   center=<3.4058 -0.4009 1.1859>
   rotation=<rot 0 86.9963 -60.5039>
  }]
 dyn_snake_hair_d01 [{
   shapeType=Capsule
   length=4.5655
   radius=3.3436
   shapeName=dyn_snake_hair_d01_capsule
   center=<2.4448 -0.6045 0.379>
   rotation=<rot 0 73.5548 -74.7868>
  }]
 dyn_snake_hair_d01_a_01 [{
   shapeType=Capsule
   length=1
   radius=2.3431
   shapeName=dyn_snake_hair_d01_a_01_capsule
   center=<-0.7771 0.6893 -0.4653>
   rotation=<rot 0 -1.845 -54.6853>
  }]
 dyn_snake_hair_d01_a_02 [{
   shapeType=Capsule
   length=0.3639
   radius=3.0992
   shapeName=dyn_snake_hair_d01_a_02_capsule
   center=<-0.9346 -0.2105 -0.8423>
   rotation=<rot 0 110.4045 -68.1481>
  }]
 dyn_snake_hair_d01_a_03 [{
   shapeType=Capsule
   length=1.4066
   radius=2.8989
   shapeName=dyn_snake_hair_d01_a_03_capsule
   center=<1.335 0.0046 -0.2399>
   rotation=<rot 0 90.5145 -78.1455>
  }]
 dyn_snake_hair_d01_a_04 [{
   shapeType=Capsule
   length=0.9972
   radius=3.1016
   shapeName=dyn_snake_hair_d01_a_04_capsule
   center=<2.582 0.1349 -0.2783>
   rotation=<rot 0 93.8195 -80.7712>
  }]
 dyn_snake_hair_d01_b_01 [{
   shapeType=Capsule
   length=2.0863
   radius=1.341
   shapeName=dyn_snake_hair_d01_b_01_capsule
   center=<1.9287 -0.1352 0.3633>
   rotation=<rot 0 86.9948 -66.3827>
  }]
 dyn_snake_hair_d01_b_02 [{
   shapeType=Capsule
   length=2.1577
   radius=1.294
   shapeName=dyn_snake_hair_d01_b_02_capsule
   center=<1.808 -0.1955 0.2014>
   rotation=<rot 0 89.5147 -82.6398>
  }]
 dyn_snake_hair_d01_b_03 [{
   shapeType=Capsule
   length=2.5049
   radius=1.5765
   shapeName=dyn_snake_hair_d01_b_03_capsule
   center=<2.2724 -0.2582 -0.4866>
   rotation=<rot 0 90.6024 -81.5982>
  }]
 dyn_snake_hair_e01 [{
   shapeType=Capsule
   length=4.472
   radius=1.6262
   shapeName=dyn_snake_hair_e01_capsule
   center=<4.8618 0.0835 -0.2105>
   rotation=<rot 0 -97.1919 -77.3906>
  }]
 dyn_snake_hair_e02 [{
   shapeType=Capsule
   length=0.9517
   radius=3.2753
   shapeName=dyn_snake_hair_e02_capsule
   center=<4.5356 0.1471 0.1085>
   rotation=<rot -24.351 -90 0.8708>
  }]
 dyn_snake_hair_e03 [{shapeType=Capsule, length=2.7762, radius=3.184, shapeName=dyn_snake_hair_e03_capsule, rotation=<rot 0 -90 -90>}]
 dyn_snake_hair_e04 [{
   shapeType=Capsule
   length=5
   radius=2.989
   shapeName=dyn_snake_hair_e04_capsule
   center=<2.113 0.4769 0.3068>
   rotation=<rot 0 93.7411 -80.0172>
  }]
 dyn_snake_hair_f01 [{
   shapeType=Capsule
   length=5.6264
   radius=1.9971
   shapeName=dyn_snake_hair_f01_capsule
   center=<5.8239 0.0651 -0.2602>
   rotation=<rot 0 -95.0872 -85.4883>
  }]
 dyn_snake_hair_f02 [{shapeType=Capsule, length=0.505, radius=0.505, shapeName=dyn_snake_hair_f02_capsule, rotation=<rot 0 -90 0>}]
 dyn_snake_hair_f03 [{
   shapeType=Capsule
   length=1.8031
   radius=1.5819
   shapeName=dyn_snake_hair_f03_capsule
   center=<1.5195 0.0411 0.0309>
   rotation=<rot -19.5555 -90 -73.7321>
  }]
 dyn_snake_hair_f04 [{
   shapeType=Capsule
   length=3.0194
   radius=1.6883
   shapeName=dyn_snake_hair_f04_capsule
   center=<2.4378 0.2997 -0.4372>
   rotation=<rot 0 89.4863 -77.2494>
  }]
 dyn_snake_hair_fl01 [{
   shapeType=Capsule
   length=2.1771
   radius=0.5879
   shapeName=dyn_snake_hair_fl01_capsule
   center=<1.9476 0.1847 -0.1225>
   rotation=<rot 0 98.6502 -88.7976>
  }]
 dyn_snake_hair_fl02 [{
   shapeType=Capsule
   length=1.8061
   radius=0.5496
   shapeName=dyn_snake_hair_fl02_capsule
   center=<1.7512 0.2087 0.1753>
   rotation=<rot 0 -93.9639 -87.7396>
  }]
 dyn_snake_hair_fl03 [{
   shapeType=Capsule
   length=1.273
   radius=0.4226
   shapeName=dyn_snake_hair_fl03_capsule
   center=<1.2682 -0.0166 -0.1286>
   rotation=<rot 0 -84.7453 -86.5658>
  }]
 dyn_snake_hair_fl04 [{
   shapeType=Capsule
   length=1.0126
   radius=0.5846
   shapeName=dyn_snake_hair_fl04_capsule
   center=<1.1228 0.0056 0.103>
   rotation=<rot 0 81.7639 -64.676>
  }]
 dyn_snake_hair_fr01 [{
   shapeType=Capsule
   length=1.3015
   radius=0.5629
   shapeName=dyn_snake_hair_fr01_capsule
   center=<2.1479 -0.1336 0.06>
   rotation=<rot 0 -79.114 -87.733>
  }]
 dyn_snake_hair_fr02 [{
   shapeType=Capsule
   length=1.7703
   radius=0.5382
   shapeName=dyn_snake_hair_fr02_capsule
   center=<2.1774 -0.0472 -0.0744>
   rotation=<rot 0 96.8609 -87.9959>
  }]
 dyn_snake_hair_fr03 [{
   shapeType=Capsule
   length=1.5875
   radius=0.4899
   shapeName=dyn_snake_hair_fr03_capsule
   center=<1.5297 0.2044 -0.033>
   rotation=<rot 0 -105.4813 -88.5663>
  }]
 dyn_snake_hair_fr04 [{
   shapeType=Capsule
   length=1.5503
   radius=0.5942
   shapeName=dyn_snake_hair_fr04_capsule
   center=<1.755 0.0138 0.0436>
   rotation=<rot 0 97.2278 -82.4439>
  }]
 dyn_snake_hair_g01 [{shapeType=Capsule, length=0.505, radius=0.505, shapeName=dyn_snake_hair_g01_capsule, rotation=<rot 0 -90 0>}]
 dyn_snake_hair_g01_a [{
   shapeType=Capsule
   length=2.2146
   radius=1.3929
   shapeName=dyn_snake_hair_g01_a_capsule
   center=<5.8883 0.2255 0.3094>
   rotation=<rot 0 -91.4602 -79.3102>
  }]
 dyn_snake_hair_g01_b [{
   shapeType=Capsule
   length=2.2248
   radius=1.4811
   shapeName=dyn_snake_hair_g01_b_capsule
   center=<1.2249 -0.0675 0.0136>
   rotation=<rot 0 96.2499 -76.0473>
  }]
 dyn_snake_hair_g02_a [{
   shapeType=Capsule
   length=2.2764
   radius=1.6881
   shapeName=dyn_snake_hair_g02_a_capsule
   center=<3.8723 0.3469 0.3309>
   rotation=<rot 0 -96.1117 -64.0032>
  }]
 dyn_snake_hair_g02_b [{
   shapeType=Capsule
   length=1.65
   radius=1.5132
   shapeName=dyn_snake_hair_g02_b_capsule
   center=<1.3877 -0.0724 0.3819>
   rotation=<rot 0 86.5609 -60.6196>
  }]
 dyn_snake_hair_g03_a [{shapeType=Capsule, length=0.505, radius=0.505, shapeName=dyn_snake_hair_g03_a_capsule, rotation=<rot 0 -90 0>}]
 dyn_snake_hair_g03_b [{
   shapeType=Capsule
   length=3.6392
   radius=1.5357
   shapeName=dyn_snake_hair_g03_b_capsule
   center=<3.2156 -0.1519 0.5576>
   rotation=<rot 0 88.6287 -78.0955>
  }]
 dyn_snake_hair_g04_a [{
   shapeType=Capsule
   length=2.9039
   radius=1.5682
   shapeName=dyn_snake_hair_g04_a_capsule
   center=<1 0.1471 0.0874>
   rotation=<rot 0 89.0396 -77.6632>
  }]
 dyn_snake_hair_h01 [{
   shapeType=Capsule
   length=0.505
   radius=0.505
   shapeName=dyn_snake_hair_h01_capsule
   center=<0.8029 -1.0659 0.9491>
   rotation=<rot 0 -61.5226 -76.5602>
  }]
 dyn_snake_hair_h01_a [{
   shapeType=Capsule
   length=1.378
   radius=1.3623
   shapeName=dyn_snake_hair_h01_a_capsule
   center=<0.5472 -0.167 -0.3644>
   rotation=<rot 0 90.7294 -83.6051>
  }]
 dyn_snake_hair_h02 [{
   shapeType=Capsule
   length=4.8129
   radius=1.4955
   shapeName=dyn_snake_hair_h02_capsule
   center=<6.5402 0.2085 -0.6>
   rotation=<rot 0 -88.3269 -82.4751>
  }]
 dyn_snake_hair_h02_a [{
   shapeType=Capsule
   length=2.4268
   radius=1.3801
   shapeName=dyn_snake_hair_h02_a_capsule
   center=<1.3662 -0.0764 -0.3362>
   rotation=<rot 0 87.3626 -79.3653>
  }]
 dyn_snake_hair_h03 [{shapeType=Capsule, length=0.505, radius=0.505, shapeName=dyn_snake_hair_h03_capsule, rotation=<rot 0 -90 0>}]
 dyn_snake_hair_h03_a [{
   shapeType=Capsule
   length=1.1844
   radius=2.8763
   shapeName=dyn_snake_hair_h03_a_capsule
   center=<2.7912 0.126 0.8266>
   rotation=<rot 0 94.3664 -54.7052>
  }]
 dyn_snake_hair_h04 [{
   shapeType=Capsule
   length=0.642
   radius=2.425
   shapeName=dyn_snake_hair_h04_capsule
   center=<2.5517 -0.1694 0.3806>
   rotation=<rot 0 -50.8549 -62.491>
  }]
 dyn_snake_hair_h05 [{
   shapeType=Capsule
   length=2.8252
   radius=1.7144
   shapeName=dyn_snake_hair_h05_capsule
   center=<3.9717 0.0011 -0.4847>
   rotation=<rot 0 91.7689 -87.236>
  }]
 dyn_snake_hair_i01 [{
   shapeType=Capsule
   length=3.1947
   radius=1.593
   shapeName=dyn_snake_hair_i01_capsule
   center=<-0.4009 -0.3484 -0.713>
   rotation=<rot 0 97.5836 -68.929>
  }]
 dyn_snake_hair_i02 [{
   shapeType=Capsule
   length=2.0903
   radius=1.3811
   shapeName=dyn_snake_hair_i02_capsule
   center=<1.2894 -0.1792 0.0819>
   rotation=<rot 0 86.5062 -85.2477>
  }]
 dyn_snake_hair_i03 [{
   shapeType=Capsule
   length=2.9498
   radius=1.6198
   shapeName=dyn_snake_hair_i03_capsule
   center=<2.0862 0.0233 0.0247>
   rotation=<rot 0 93.2038 -82.2777>
  }]
 dyn_snake_hair_j01 [{
   shapeType=Capsule
   length=5.0943
   radius=2.0166
   shapeName=dyn_snake_hair_j01_capsule
   center=<3.4659 0.0253 -0.1978>
   rotation=<rot 0 90.3278 -88.8149>
  }]
 dyn_snake_hair_j02 [{shapeType=Capsule, length=0.505, radius=0.505, shapeName=dyn_snake_hair_j02_capsule, rotation=<rot 0 -90 0>}]
 dyn_snake_hair_j03 [{
   shapeType=Capsule
   length=3.1935
   radius=1.6846
   shapeName=dyn_snake_hair_j03_capsule
   center=<2.6427 0.0434 -0.0908>
   rotation=<rot 0 96.1223 -87.1963>
  }]
 dyn_snake_hair_k01 [{
   shapeType=Capsule
   length=2.116
   radius=1.6995
   shapeName=dyn_snake_hair_k01_capsule
   center=<1.5391 0.0785 0.2229>
   rotation=<rot 0 -99.4404 -89.0869>
  }]
 dyn_snake_hair_k02 [{
   shapeType=Capsule
   length=1.5989
   radius=1.3408
   shapeName=dyn_snake_hair_k02_capsule
   center=<0.6157 0.0804 0.245>
   rotation=<rot 0 -149.1636 -76.5215>
  }]
 dyn_snake_hair_k03 [{
   shapeType=Capsule
   length=3.3444
   radius=1.7384
   shapeName=dyn_snake_hair_k03_capsule
   center=<0.9379 0.2752 -0.0528>
   rotation=<rot 0 95.8907 -83.1221>
  }]
 head [{shapeType=Capsule, length=4.219, radius=11.4363, shapeName=head_capsule, center=<4.4631 -0.0004 -0.1218>, rotation=<rot 88.4366 -0.005 0>}]
 neck_01 [{
   shapeType=Capsule
   length=11.0088
   radius=5.5025
   shapeName=neck_01_capsule
   center=<3.6064 -0.0006 0.0513>
   rotation=<rot 89.1855 179.9905 -180>
  }]
 spine_05 [{shapeType=Capsule, length=11.373, radius=13.8726, shapeName=spine_05_capsule, center=<0 -1.6731 0>}]
 upperarm_l [{shapeType=Capsule, length=17.5972, radius=5.7088, shapeName=upperarm_l_capsule, center=<12.0726 0 0>, rotation=<rot -90 179.7626 -179.7626>}]
 upperarm_r [{shapeType=Capsule, length=18.3472, radius=4.9588, shapeName=upperarm_r_capsule, center=<-12.0645 0 0>, rotation=<rot -90 144.7356 -144.7356>}]
}
constraints [
 {bone1Name=dyn_snake_hair_a01, bone2Name=head, swing1LimitDegrees=25, swing2LimitDegrees=25, twistLimitDegrees=5}
 {bone1Name=dyn_snake_hair_a02, bone2Name=dyn_snake_hair_a01, swing1LimitDegrees=25, swing2LimitDegrees=25, twistLimitDegrees=5}
 {bone1Name=dyn_snake_hair_a03, bone2Name=dyn_snake_hair_a02, swing1LimitDegrees=25, swing2LimitDegrees=25, twistLimitDegrees=5}
 {bone1Name=dyn_snake_hair_a04, bone2Name=dyn_snake_hair_a03, swing1LimitDegrees=25, swing2LimitDegrees=25, twistLimitDegrees=5}
 {bone1Name=dyn_snake_hair_b01, bone2Name=head, swing1LimitDegrees=25, swing2LimitDegrees=25, twistLimitDegrees=5}
 {bone1Name=dyn_snake_hair_b02, bone2Name=dyn_snake_hair_b01, swing1LimitDegrees=25, swing2LimitDegrees=25, twistLimitDegrees=5}
 {bone1Name=dyn_snake_hair_b03, bone2Name=dyn_snake_hair_b02, swing1LimitDegrees=25, swing2LimitDegrees=25, twistLimitDegrees=5}
 {bone1Name=dyn_snake_hair_b04, bone2Name=dyn_snake_hair_b03, swing1LimitDegrees=25, swing2LimitDegrees=25, twistLimitDegrees=5}
 {bone1Name=dyn_snake_hair_c01, bone2Name=head, swing1LimitDegrees=25, swing2LimitDegrees=25, twistLimitDegrees=5}
 {bone1Name=dyn_snake_hair_c02, bone2Name=dyn_snake_hair_c01, swing1LimitDegrees=25, swing2LimitDegrees=25, twistLimitDegrees=5}
 {bone1Name=dyn_snake_hair_c03, bone2Name=dyn_snake_hair_c02, swing1LimitDegrees=25, swing2LimitDegrees=25, twistLimitDegrees=5}
 {bone1Name=dyn_snake_hair_c04, bone2Name=dyn_snake_hair_c03, swing1LimitDegrees=25, swing2LimitDegrees=25, twistLimitDegrees=5}
 {bone1Name=dyn_snake_hair_d01, bone2Name=head, swing1LimitDegrees=25, swing2LimitDegrees=25, twistLimitDegrees=5}
 {bone1Name=dyn_snake_hair_d01_a_01, bone2Name=dyn_snake_hair_d01, swing1LimitDegrees=25, swing2LimitDegrees=25, twistLimitDegrees=5}
 {bone1Name=dyn_snake_hair_d01_a_02, bone2Name=dyn_snake_hair_d01_a_01, swing1LimitDegrees=25, swing2LimitDegrees=25, twistLimitDegrees=5}
 {bone1Name=dyn_snake_hair_d01_a_03, bone2Name=dyn_snake_hair_d01_a_02, swing1LimitDegrees=25, swing2LimitDegrees=25, twistLimitDegrees=5}
 {bone1Name=dyn_snake_hair_d01_a_04, bone2Name=dyn_snake_hair_d01_a_03, swing1LimitDegrees=25, swing2LimitDegrees=25, twistLimitDegrees=5}
 {bone1Name=dyn_snake_hair_d01_b_01, bone2Name=dyn_snake_hair_d01, swing1LimitDegrees=25, swing2LimitDegrees=25, twistLimitDegrees=5}
 {bone1Name=dyn_snake_hair_d01_b_02, bone2Name=dyn_snake_hair_d01_b_01, swing1LimitDegrees=25, swing2LimitDegrees=25, twistLimitDegrees=5}
 {bone1Name=dyn_snake_hair_d01_b_03, bone2Name=dyn_snake_hair_d01_b_02, swing1LimitDegrees=25, swing2LimitDegrees=25, twistLimitDegrees=5}
 {bone1Name=dyn_snake_hair_e01, bone2Name=head, swing1LimitDegrees=25, swing2LimitDegrees=25, twistLimitDegrees=5}
 {bone1Name=dyn_snake_hair_e02, bone2Name=dyn_snake_hair_e01, swing1LimitDegrees=25, swing2LimitDegrees=25, twistLimitDegrees=5}
 {bone1Name=dyn_snake_hair_e03, bone2Name=dyn_snake_hair_e02, swing1LimitDegrees=25, swing2LimitDegrees=25, twistLimitDegrees=5}
 {bone1Name=dyn_snake_hair_e04, bone2Name=dyn_snake_hair_e03, swing1LimitDegrees=25, swing2LimitDegrees=25, twistLimitDegrees=5}
 {bone1Name=dyn_snake_hair_f01, bone2Name=head, swing1LimitDegrees=25, swing2LimitDegrees=25, twistLimitDegrees=5}
 {bone1Name=dyn_snake_hair_f02, bone2Name=dyn_snake_hair_f01, swing1LimitDegrees=25, swing2LimitDegrees=25, twistLimitDegrees=5}
 {bone1Name=dyn_snake_hair_f03, bone2Name=dyn_snake_hair_f02, swing1LimitDegrees=25, swing2LimitDegrees=25, twistLimitDegrees=5}
 {bone1Name=dyn_snake_hair_f04, bone2Name=dyn_snake_hair_f03, swing1LimitDegrees=25, swing2LimitDegrees=25, twistLimitDegrees=5}
 {bone1Name=dyn_snake_hair_g01, bone2Name=head, swing1LimitDegrees=25, swing2LimitDegrees=25, twistLimitDegrees=5}
 {bone1Name=dyn_snake_hair_g01_a, bone2Name=dyn_snake_hair_g01, swing1LimitDegrees=25, swing2LimitDegrees=25, twistLimitDegrees=5}
 {bone1Name=dyn_snake_hair_g02_a, bone2Name=dyn_snake_hair_g01_a, swing1LimitDegrees=25, swing2LimitDegrees=25, twistLimitDegrees=5}
 {bone1Name=dyn_snake_hair_g03_a, bone2Name=dyn_snake_hair_g02_a, swing1LimitDegrees=25, swing2LimitDegrees=25, twistLimitDegrees=5}
 {bone1Name=dyn_snake_hair_g04_a, bone2Name=dyn_snake_hair_g03_a, swing1LimitDegrees=25, swing2LimitDegrees=25, twistLimitDegrees=5}
 {bone1Name=dyn_snake_hair_g01_b, bone2Name=dyn_snake_hair_g01, swing1LimitDegrees=25, swing2LimitDegrees=25, twistLimitDegrees=5}
 {bone1Name=dyn_snake_hair_g02_b, bone2Name=dyn_snake_hair_g01_b, swing1LimitDegrees=25, swing2LimitDegrees=25, twistLimitDegrees=5}
 {bone1Name=dyn_snake_hair_g03_b, bone2Name=dyn_snake_hair_g02_b, swing1LimitDegrees=25, swing2LimitDegrees=25, twistLimitDegrees=5}
 {bone1Name=dyn_snake_hair_h01, bone2Name=head, swing1LimitDegrees=25, swing2LimitDegrees=25, twistLimitDegrees=5}
 {bone1Name=dyn_snake_hair_h02, bone2Name=dyn_snake_hair_h01, swing1LimitDegrees=25, swing2LimitDegrees=25, twistLimitDegrees=5}
 {bone1Name=dyn_snake_hair_h03, bone2Name=dyn_snake_hair_h02, swing1LimitDegrees=25, swing2LimitDegrees=25, twistLimitDegrees=5}
 {bone1Name=dyn_snake_hair_h04, bone2Name=dyn_snake_hair_h03, swing1LimitDegrees=25, swing2LimitDegrees=25, twistLimitDegrees=5}
 {bone1Name=dyn_snake_hair_h05, bone2Name=dyn_snake_hair_h04, swing1LimitDegrees=25, swing2LimitDegrees=25, twistLimitDegrees=5}
 {
  bone1Name=dyn_snake_hair_fl01
  bone2Name=head
  swing1LimitDegrees=45
  swing1Motion=Free
  swing2LimitDegrees=45
  swing2Motion=Free
  twistLimitDegrees=45
  twistMotion=Free
 }
 {
  bone1Name=dyn_snake_hair_fl02
  bone2Name=dyn_snake_hair_fl01
  swing1LimitDegrees=45
  swing1Motion=Free
  swing2LimitDegrees=45
  swing2Motion=Free
  twistLimitDegrees=45
  twistMotion=Free
 }
 {
  bone1Name=dyn_snake_hair_fl03
  bone2Name=dyn_snake_hair_fl02
  swing1LimitDegrees=45
  swing1Motion=Free
  swing2LimitDegrees=45
  swing2Motion=Free
  twistLimitDegrees=45
  twistMotion=Free
 }
 {
  bone1Name=dyn_snake_hair_fl04
  bone2Name=dyn_snake_hair_fl03
  swing1LimitDegrees=45
  swing1Motion=Free
  swing2LimitDegrees=45
  swing2Motion=Free
  twistLimitDegrees=45
  twistMotion=Free
 }
 {
  bone1Name=dyn_snake_hair_fr01
  bone2Name=head
  swing1LimitDegrees=45
  swing1Motion=Free
  swing2LimitDegrees=45
  swing2Motion=Free
  twistLimitDegrees=45
  twistMotion=Free
 }
 {
  bone1Name=dyn_snake_hair_fr02
  bone2Name=dyn_snake_hair_fr01
  swing1LimitDegrees=45
  swing1Motion=Free
  swing2LimitDegrees=45
  swing2Motion=Free
  twistLimitDegrees=45
  twistMotion=Free
 }
 {
  bone1Name=dyn_snake_hair_fr03
  bone2Name=dyn_snake_hair_fr02
  swing1LimitDegrees=45
  swing1Motion=Free
  swing2LimitDegrees=45
  swing2Motion=Free
  twistLimitDegrees=45
  twistMotion=Free
 }
 {
  bone1Name=dyn_snake_hair_fr04
  bone2Name=dyn_snake_hair_fr03
  swing1LimitDegrees=45
  swing1Motion=Free
  swing2LimitDegrees=45
  swing2Motion=Free
  twistLimitDegrees=45
  twistMotion=Free
 }
 {bone1Name=dyn_snake_hair_h01_a, bone2Name=dyn_snake_hair_h01, swing1LimitDegrees=6, swing2LimitDegrees=12, twistLimitDegrees=5}
 {bone1Name=dyn_snake_hair_h02_a, bone2Name=dyn_snake_hair_h01_a, swing1LimitDegrees=6, swing2LimitDegrees=12, twistLimitDegrees=5}
 {bone1Name=dyn_snake_hair_h03_a, bone2Name=dyn_snake_hair_h02_a, swing1LimitDegrees=6, swing2LimitDegrees=12, twistLimitDegrees=5}
 {bone1Name=dyn_snake_hair_i01, bone2Name=head, swing1LimitDegrees=6, swing2LimitDegrees=12, twistLimitDegrees=5}
 {bone1Name=dyn_snake_hair_i02, bone2Name=dyn_snake_hair_i01, swing1LimitDegrees=6, swing2LimitDegrees=12, twistLimitDegrees=5}
 {bone1Name=dyn_snake_hair_i03, bone2Name=dyn_snake_hair_i02, swing1LimitDegrees=6, swing2LimitDegrees=12, twistLimitDegrees=5}
 {bone1Name=dyn_snake_hair_j01, bone2Name=head, swing1LimitDegrees=6, swing2LimitDegrees=12, twistLimitDegrees=5}
 {bone1Name=dyn_snake_hair_j02, bone2Name=dyn_snake_hair_j01, swing1LimitDegrees=6, swing2LimitDegrees=12, twistLimitDegrees=5}
 {bone1Name=dyn_snake_hair_j03, bone2Name=dyn_snake_hair_j02, swing1LimitDegrees=6, swing2LimitDegrees=12, twistLimitDegrees=5}
 {bone1Name=dyn_snake_hair_k01, bone2Name=head, swing1LimitDegrees=6, swing2LimitDegrees=12, twistLimitDegrees=5}
 {bone1Name=dyn_snake_hair_k02, bone2Name=dyn_snake_hair_k01, swing1LimitDegrees=6, swing2LimitDegrees=12, twistLimitDegrees=5}
 {bone1Name=dyn_snake_hair_k03, bone2Name=dyn_snake_hair_k02, swing1LimitDegrees=6, swing2LimitDegrees=12, twistLimitDegrees=5}
 {
  bone1Name=upperarm_l
  bone2Name=clavicle_l
  swing1LimitDegrees=45
  swing1Motion=Free
  swing2LimitDegrees=45
  swing2Motion=Free
  twistLimitDegrees=45
  twistMotion=Free
 }
 {
  bone1Name=clavicle_l
  bone2Name=spine_05
  swing1LimitDegrees=45
  swing1Motion=Free
  swing2LimitDegrees=45
  swing2Motion=Free
  twistLimitDegrees=45
  twistMotion=Free
 }
]
