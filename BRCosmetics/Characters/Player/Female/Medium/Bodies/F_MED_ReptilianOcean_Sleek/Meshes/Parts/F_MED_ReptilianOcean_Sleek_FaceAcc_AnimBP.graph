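# Anim Blueprint for the ReptilianOcean Sleek female character's FaceAcc part, animating her snake hair.
# The copied parent pose runs through a Control Rig, AnimDynamics and a RigidBody node, then gets idle/ADS and tongue additives.
# A layered blend mixes in frontend custom-idle slot animation by curve value, before a targeting Control Rig that accounts for backbling.
# Two RigidBodyWithControl setups sit outside the output chain, and the computed gravity override has no visible consumer.
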AnimBlueprint /BRCosmetics/Characters/Player/Female/Medium/Bodies/F_MED_ReptilianOcean_Sleek/Meshes/Parts/F_MED_ReptilianOcean_Sleek_FaceAcc_AnimBP
target_skeleton=/Game/Characters/Player/Female/Medium/Bodies/F_MED_ReptilianOcean_Sleek/Meshes/Parts/F_MED_ReptilianOcean_Sleek_FaceAcc_Skeleton
graphs [
 AnimGraph {nodes [
   n1 = RigidBodyWithControl!(
    ComponentPose:Component Space Pose Link Structure=Local To Component()
    Alpha:float=1
    ExternalForce:Vector="0,0,0"
    ControlAndModifierParameters:Physics Control Control And Modifier Parameters Structure=(ControlParameters=,ControlMultiplierParameters=,ModifierParameters=)
    ControlAndModifierUpdates:Physics Control Control And Modifier Updates Structure=(ControlUpdates=,ControlMultiplierUpdates=,ModifierUpdates=)
    ControlTargets:Rigid Body Control Targets Structure=(Targets=())
    KinematicTargets:Rigid Body Kinematic Targets Structure=(Targets=())
    ForcesAndTorques:Rigid Body Forces And Torques Structure=(ForcesAndTorques=())
   ) -> (Pose:Component Space Pose Link Structure) {
    is=title AnimGraphNode_RigidBodyWithControl
    bOverrideWorldGravity=true
    overridePhysicsAsset=/Game/Characters/Player/Female/Medium/Bodies/F_MED_ReptilianOcean_Sleek/Meshes/Parts/F_MED_ReptilianOcean_Sleek_FaceAcc_Physics_RBWC
    additionalSets {controlSetUpdates [{setName=ParentSpace_Hair, names [Hair]}], modifierSetUpdates [{setName=HairModSet, names [Hair]}]}
    alphaBoolBlend {blendOption=Linear}
    characterSetupData {limbSetupData [{
       bCreateBodyModifiers=true
       bCreateParentSpaceControls=true
       bCreateWorldSpaceControls=true
       bIncludeParentBone=false
       limbName=Hair
       startBone=dyn_snake_hair_k01
      }]}
    componentAppliedLinearAccClamp=<0 0 0>
    initialControlAndBodyModifierUpdates {
     controlParameters [ParentSpace_Hair {data {
        angularDampingRatio=0.5
        angularExtraDamping=0
        angularStrength=40
        angularTargetVelocityMultiplier=1
        bDisableCollision=false
        bEnableAngularDampingRatio=true
        bEnableAngularExtraDamping=true
        bEnableAngularStrength=true
        bEnableAngularTargetVelocityMultiplier=true
        bEnableCustomControlPoint=true
        bEnableLinearDampingRatio=true
        bEnableLinearExtraDamping=true
        bEnableLinearStrength=true
        bEnableLinearTargetVelocityMultiplier=true
        bEnableMaxForce=true
        bEnableMaxTorque=true
        bEnablebDisableCollision=true
        bEnablebEnabled=true
        bEnablebOnlyControlChildObject=true
        bEnablebUseAccelerationDriveMode=true
        bEnablebUseCustomControlPoint=true
        bEnablebUseSkeletalAnimation=true
        bEnabled=true
        bOnlyControlChildObject=false
        bUseAccelerationDriveMode=true
        bUseCustomControlPoint=false
        bUseSkeletalAnimation=false
        linearDampingRatio=0.5
        linearExtraDamping=0
        linearStrength=0
        linearTargetVelocityMultiplier=1
        maxForce=0
        maxTorque=0
        customControlPoint=<0 0 0>
       }}]
     modifierParameters [HairModSet {data {
        bEnableCCD=false
        bEnableCollisionType=true
        bEnableGravityMultiplier=true
        bEnableKinematicTargetSpace=true
        bEnableMovementType=true
        bEnablePhysicsBlendWeight=true
        bEnablebEnableCCD=true
        bEnablebUpdateKinematicFromSimulation=true
        bUpdateKinematicFromSimulation=true
        collisionType=QueryAndPhysics
        gravityMultiplier=0
        kinematicTargetSpace=OffsetInBoneSpace
        movementType=Simulated
        physicsBlendWeight=1
       }}]
    }
    simSpaceSettings {
     maxAngularAcceleration=1
     maxAngularVelocity=0.01
     maxLinearAcceleration=0.5
     maxLinearVelocity=20000
     worldAlpha=1
     externalLinearDragV=<0.5 0.5>
    }
   }
   Component To Local(ComponentPose=@n1)
   RigidBodyWithControl!(ComponentPose, Alpha=1, OverrideWorldGravity:Vector="0,0,0", ExternalForce="0,0,0", ForcesAndTorques=(ForcesAndTorques=())) -> (Pose) {
    bEnableControls=true
    bOverrideWorldGravity=true
    evaluationResetTime=0.2
    overridePhysicsAsset=&F_MED_ReptilianOcean_Sleek_FaceAcc_Physics_RBWC
    additionalSets {controlSetUpdates [{setName=ParentSpace_Hair, names [Hair]}], modifierSetUpdates [{setName=HairModSet, names [Hair]}]}
    alphaBoolBlend {blendOption=Linear}
    characterSetupData {
     defaultParentSpaceControlData {bEnabled=false}
     defaultWorldSpaceControlData {bEnabled=false}
     limbSetupData [{
       bCreateBodyModifiers=true
       bCreateParentSpaceControls=true
       bCreateWorldSpaceControls=true
       bIncludeParentBone=false
       limbName=Hair
       startBone=dyn_snake_hair_k01
      }]
    }
    componentAppliedLinearAccClamp=<0 0 0>
    controlAndModifierParameters {
     controlParameters [ParentSpace_Hair {data {
        angularDampingRatio=0.5
        angularExtraDamping=0
        angularStrength=40
        angularTargetVelocityMultiplier=1
        bDisableCollision=false
        bEnableAngularDampingRatio=true
        bEnableAngularExtraDamping=true
        bEnableAngularStrength=true
        bEnableAngularTargetVelocityMultiplier=true
        bEnableCustomControlPoint=true
        bEnableLinearDampingRatio=true
        bEnableLinearExtraDamping=true
        bEnableLinearStrength=true
        bEnableLinearTargetVelocityMultiplier=true
        bEnableMaxForce=true
        bEnableMaxTorque=true
        bEnablebDisableCollision=true
        bEnablebEnabled=true
        bEnablebOnlyControlChildObject=true
        bEnablebUseAccelerationDriveMode=true
        bEnablebUseCustomControlPoint=true
        bEnablebUseSkeletalAnimation=true
        bEnabled=true
        bOnlyControlChildObject=false
        bUseAccelerationDriveMode=true
        bUseCustomControlPoint=false
        bUseSkeletalAnimation=false
        linearDampingRatio=0.5
        linearExtraDamping=0
        linearStrength=0
        linearTargetVelocityMultiplier=1
        maxForce=0
        maxTorque=0
        customControlPoint=<0 0 0>
       }}]
     modifierParameters [HairModSet {data {
        bEnableCCD=false
        bEnableCollisionType=true
        bEnableGravityMultiplier=true
        bEnableKinematicTargetSpace=true
        bEnableMovementType=true
        bEnablePhysicsBlendWeight=true
        bEnablebEnableCCD=true
        bEnablebUpdateKinematicFromSimulation=true
        bUpdateKinematicFromSimulation=true
        collisionType=QueryAndPhysics
        gravityMultiplier=0
        kinematicTargetSpace=OffsetInBoneSpace
        movementType=Simulated
        physicsBlendWeight=1
       }}]
    }
    simSpaceSettings {
     maxAngularAcceleration=1
     maxAngularVelocity=0.01
     maxLinearAcceleration=0.5
     maxLinearVelocity=20000
     worldAlpha=1
     externalLinearDragV=<0.5 0.5>
    }
   }
   AnimGraphNode_Slot_customIdle_0 = Slot 'DefaultSlot'()
   AnimGraphNode_UseCachedPose_customIdle_0 = Use cached pose 'startpose'()
   K2Node_VariableGet_customIdle_1 = Get Custom Idle is Playing() -> (Custom Idle is Playing:float)
   AnimGraphNode_LayeredBoneBlend_customIdle_0 = Layered blend per bone(
    BasePose=@AnimGraphNode_UseCachedPose_customIdle_0
    BlendPoses_0=@AnimGraphNode_Slot_customIdle_0
    BlendWeights_0=@K2Node_VariableGet_customIdle_1
   ) {layerSetup [{branchFilters [{boneName=dyn_snake_hair_k01}, {boneName=dyn_snake_hair_j01}, {boneName=dyn_snake_hair_i01}, {boneName=dyn_snake_hair_h01}]}]}
   n2 = Control Rig!(
    Source=@AnimGraphNode_LayeredBoneBlend_customIdle_0
    Alpha
    IsWearingBackling:bool=Get Is Wearing Backbling()
    IsTargeting:bool=Get bIsTargeting()
   ) -> (Pose) {
    alphaBoolBlend {blendOption=Linear}
    controlRigAssetReference {blueprintRigClass=/Game/Characters/Player/Female/Medium/Bodies/F_MED_ReptilianOcean_Sleek/Meshes/Parts/F_MED_ReptilianOcean_Sleek_FaceAcc_Targeting_CtrlRig.F_MED_ReptilianOcean_Sleek_FaceAcc_Targeting_CtrlRig_C}
   }
   Output Pose(Result=@n2) {layerGroup=None}
   n3 = Control Rig!(Source=Copy Pose From Mesh() {bUseAttachedParent=true}, Alpha, IsTargeting=Get bIsTargeting()) -> (Pose) {
    alphaBoolBlend {blendOption=Linear}
    controlRigAssetReference {blueprintRigClass=&/F_MED_ReptilianOcean_Sleek_FaceAcc_CtrlRig.F_MED_ReptilianOcean_Sleek_FaceAcc_CtrlRig_C}
   }
   n4 = Anim Dynamics - Bone: dyn_snake_hair_fl01(ComponentPose=Local To Component(LocalPose=@n3), Alpha=1, ExternalForce="0,0,0") {
    angularDampingOverride=1
    angularSpringConstant=20000
    bAngularSpring=true
    bOverrideAngularDamping=true
    bUseGravityOverride=true
    gravityScale=0.35
    simulationSpace=RootRelative
    alphaBoolBlend {blendOption=Linear}
    boundBone {boneName=dyn_snake_hair_fl01}
    componentLinearAccScale=<1 1 1>
    componentLinearVelScale=<1 1 1>
    gravityOverride {z=-980}
    physicsBodyDefinitions [{
      boundBone {boneName=dyn_snake_hair_fl01}
      boxExtents=<20 30 30>
      constraintSetup {angularLimitsMax=<2 2 2>, angularLimitsMin=<-2 -2 -2>, angularTarget=<1 0 0>}
      localJointOffset=<4 0 0>
     }]
   }
   n5 = Anim Dynamics - Bone: dyn_snake_hair_fl02(ComponentPose=@n4, Alpha=1, ExternalForce="0,0,0") {
    angularDampingOverride=1
    angularSpringConstant=20000
    bAngularSpring=true
    bOverrideAngularDamping=true
    bUseGravityOverride=true
    gravityScale=0.35
    simulationSpace=RootRelative
    alphaBoolBlend {blendOption=Linear}
    boundBone {boneName=dyn_snake_hair_fl02}
    componentLinearAccScale=<1 1 1>
    componentLinearVelScale=<1 1 1>
    gravityOverride {z=-980}
    physicsBodyDefinitions [{
      boundBone {boneName=dyn_snake_hair_fl02}
      boxExtents=<20 30 30>
      constraintSetup {angularLimitsMax=<2 2 2>, angularLimitsMin=<-2 -2 -2>, angularTarget=<1 0 0>}
      localJointOffset=<4 0 0>
     }]
   }
   n6 = Anim Dynamics - Bone: dyn_snake_hair_fl03(ComponentPose=@n5, Alpha=1, ExternalForce="0,0,0") {
    angularDampingOverride=1
    angularSpringConstant=20000
    bAngularSpring=true
    bOverrideAngularDamping=true
    bUseGravityOverride=true
    gravityScale=0.35
    simulationSpace=RootRelative
    alphaBoolBlend {blendOption=Linear}
    boundBone {boneName=dyn_snake_hair_fl03}
    componentLinearAccScale=<1 1 1>
    componentLinearVelScale=<1 1 1>
    gravityOverride {z=-980}
    physicsBodyDefinitions [{
      boundBone {boneName=dyn_snake_hair_fl03}
      boxExtents=<20 30 30>
      constraintSetup {angularLimitsMax=<2 2 2>, angularLimitsMin=<-2 -2 -2>, angularTarget=<1 0 0>}
      localJointOffset=<4 0 0>
     }]
   }
   n7 = Anim Dynamics - Bone: dyn_snake_hair_fl04(ComponentPose=@n6, Alpha=1, ExternalForce="0,0,0") {
    angularDampingOverride=1
    angularSpringConstant=20000
    bAngularSpring=true
    bOverrideAngularDamping=true
    bUseGravityOverride=true
    gravityScale=0.35
    simulationSpace=RootRelative
    alphaBoolBlend {blendOption=Linear}
    boundBone {boneName=dyn_snake_hair_fl04}
    componentLinearAccScale=<1 1 1>
    componentLinearVelScale=<1 1 1>
    gravityOverride {z=-980}
    physicsBodyDefinitions [{
      boundBone {boneName=dyn_snake_hair_fl04}
      boxExtents=<20 30 30>
      constraintSetup {angularLimitsMax=<2 2 2>, angularLimitsMin=<-2 -2 -2>, angularTarget=<1 0 0>}
      localJointOffset=<4 0 0>
     }]
   }
   n8 = Anim Dynamics - Bone: dyn_snake_hair_fr01(ComponentPose=@n7, Alpha=1, ExternalForce="0,0,0") {
    angularDampingOverride=1
    angularSpringConstant=20000
    bAngularSpring=true
    bOverrideAngularDamping=true
    bUseGravityOverride=true
    gravityScale=0.35
    simulationSpace=RootRelative
    alphaBoolBlend {blendOption=Linear}
    boundBone {boneName=dyn_snake_hair_fr01}
    componentLinearAccScale=<1 1 1>
    componentLinearVelScale=<1 1 1>
    gravityOverride {z=-980}
    physicsBodyDefinitions [{
      boundBone {boneName=dyn_snake_hair_fr01}
      boxExtents=<20 30 30>
      constraintSetup {angularLimitsMax=<2 2 2>, angularLimitsMin=<-2 -2 -2>, angularTarget=<1 0 0>}
      localJointOffset=<4 0 0>
     }]
   }
   n9 = Anim Dynamics - Bone: dyn_snake_hair_fr02(ComponentPose=@n8, Alpha=1, ExternalForce="0,0,0") {
    angularDampingOverride=1
    angularSpringConstant=20000
    bAngularSpring=true
    bOverrideAngularDamping=true
    bUseGravityOverride=true
    gravityScale=0.35
    simulationSpace=RootRelative
    alphaBoolBlend {blendOption=Linear}
    boundBone {boneName=dyn_snake_hair_fr02}
    componentLinearAccScale=<1 1 1>
    componentLinearVelScale=<1 1 1>
    gravityOverride {z=-980}
    physicsBodyDefinitions [{
      boundBone {boneName=dyn_snake_hair_fr02}
      boxExtents=<20 30 30>
      constraintSetup {angularLimitsMax=<2 2 2>, angularLimitsMin=<-2 -2 -2>, angularTarget=<1 0 0>}
      localJointOffset=<4 0 0>
     }]
   }
   n10 = Anim Dynamics - Bone: dyn_snake_hair_fr03(ComponentPose=@n9, Alpha=1, ExternalForce="0,0,0") {
    angularDampingOverride=1
    angularSpringConstant=20000
    bAngularSpring=true
    bOverrideAngularDamping=true
    bUseGravityOverride=true
    gravityScale=0.35
    simulationSpace=RootRelative
    alphaBoolBlend {blendOption=Linear}
    boundBone {boneName=dyn_snake_hair_fr03}
    componentLinearAccScale=<1 1 1>
    componentLinearVelScale=<1 1 1>
    gravityOverride {z=-980}
    physicsBodyDefinitions [{
      boundBone {boneName=dyn_snake_hair_fr03}
      boxExtents=<20 30 30>
      constraintSetup {angularLimitsMax=<2 2 2>, angularLimitsMin=<-2 -2 -2>, angularTarget=<1 0 0>}
      localJointOffset=<4 0 0>
     }]
   }
   n11 = Anim Dynamics - Bone: dyn_snake_hair_fr04(ComponentPose=@n10, Alpha=1, ExternalForce="0,0,0") {
    angularDampingOverride=1
    angularSpringConstant=20000
    bAngularSpring=true
    bOverrideAngularDamping=true
    bUseGravityOverride=true
    gravityScale=0.35
    simulationSpace=RootRelative
    alphaBoolBlend {blendOption=Linear}
    boundBone {boneName=dyn_snake_hair_fr04}
    componentLinearAccScale=<1 1 1>
    componentLinearVelScale=<1 1 1>
    gravityOverride {z=-980}
    physicsBodyDefinitions [{
      boundBone {boneName=dyn_snake_hair_fr04}
      boxExtents=<20 30 30>
      constraintSetup {angularLimitsMax=<2 2 2>, angularLimitsMin=<-2 -2 -2>, angularTarget=<1 0 0>}
      localJointOffset=<4 0 0>
     }]
   }
   n12 = RigidBody!(ComponentPose=@n11, Alpha, OverrideWorldGravity:Vector="0,0,-980", ExternalForce) -> (Pose) {
    bClampLinearTranslationLimitToRefPose=true
    bOverrideWorldGravity=true
    overridePhysicsAsset=&/F_MED_ReptilianOcean_Sleek_FaceAcc_Physics
    simulationSpace=BaseBoneSpace
    alphaBoolBlend {blendOption=Linear}
    baseBoneRef {boneName=neck_01}
    componentAppliedLinearAccClamp=<1200 1200 1000>
    componentLinearAccScale=<0.01 0.15 0.15>
    componentLinearVelScale=<0.08 0.4 0.4>
    overrideWorldGravity {z=-980}
   }
   n13 = Component To Local(ComponentPose=@n12)
   AnimGraphNode_StateMachine_0 = Idle/ADS Hair Anim() {is=title AnimGraphNode_StateMachine}
   n14 = "Sequence Player (additive) 'ReptilianOcean_tongueAdditive_Anim_FaceAcc'"() {sequence=/BRCosmetics/Characters/Player/Female/Medium/Bodies/F_MED_ReptilianOcean_Sleek/Animation/ReptilianOcean_tongueAdditive_Anim_FaceAcc}
   AnimGraphNode_SaveCachedPose_customIdle_0 = Save cached pose 'startpose'(Pose=Apply Additive(Base=Apply Additive(Base=@n13, Additive=@AnimGraphNode_StateMachine_0) {alphaBoolBlend {blendOption=Linear}}, Additive=@n14) {alphaBoolBlend {blendOption=Linear}}) {cacheName=startpose}
   Get Control Targets() -> (Control Targets:Rigid Body Control Targets Structure)
  ]}
 Idle/ADS Hair Anim {
  entry=Idle
  path=AnimGraph.AnimGraphNode_StateMachine_0.Idle/ADS Hair Anim
  ids=id
  nodes [
   AnimStateNode_0 = Idle() {
    stateEntered {guid=#1}
    stateFullyBlended {guid=#2}
    stateLeft {guid=#3}
   }
   AnimStateNode_1 = ADS() {
    is=title AnimStateNode
    stateEntered {guid=#4}
    stateFullyBlended {guid=#5}
    stateLeft {guid=#6}
   }
   AnimStateNode_2 = ADS Moving() {
    is=title AnimStateNode
    stateEntered {guid=#7}
    stateFullyBlended {guid=#8}
    stateLeft {guid=#9}
   }
   AnimStateTransitionNode_0 = AnimStateTransitionNode() {
    blendMode=Sinusoidal
    crossfadeDuration=0.3
    transitionEnd {guid=#10}
    transitionInterrupt {guid=#11}
    transitionStart {guid=#12}
   }
   AnimStateTransitionNode_1 = AnimStateTransitionNode() {
    transitionEnd {guid=#13}
    transitionInterrupt {guid=#14}
    transitionStart {guid=#15}
   }
   AnimStateTransitionNode_2 = AnimStateTransitionNode() {
    blendMode=Sinusoidal
    transitionEnd {guid=#16}
    transitionInterrupt {guid=#17}
    transitionStart {guid=#18}
   }
   AnimStateTransitionNode_3 = AnimStateTransitionNode() {
    transitionEnd {guid=#19}
    transitionInterrupt {guid=#20}
    transitionStart {guid=#21}
   }
   AnimStateTransitionNode_4 = AnimStateTransitionNode() {crossfadeDuration=0.5}
   AnimStateTransitionNode_7 = AnimStateTransitionNode() {
    crossfadeDuration=0.5
    transitionEnd {guid=#22}
    transitionInterrupt {guid=#23}
    transitionStart {guid=#24}
   }
  ]
  links [
   {
    dst=ADS Moving
    rule=AnimGraph.AnimGraphNode_StateMachine_0.Idle/ADS Hair Anim.AnimStateTransitionNode_4.Transition
    src=ADS
    via=AnimStateTransitionNode_4
   }
   {
    dst=Idle
    rule=AnimGraph.AnimGraphNode_StateMachine_0.Idle/ADS Hair Anim.AnimStateTransitionNode_2.Transition
    src=ADS
    via=AnimStateTransitionNode_2
   }
   {
    dst=ADS
    rule=AnimGraph.AnimGraphNode_StateMachine_0.Idle/ADS Hair Anim.AnimStateTransitionNode_7.Transition
    src=ADS Moving
    via=AnimStateTransitionNode_7
   }
   {
    dst=Idle
    rule=AnimGraph.AnimGraphNode_StateMachine_0.Idle/ADS Hair Anim.AnimStateTransitionNode_1.Transition
    src=ADS Moving
    via=AnimStateTransitionNode_1
   }
   {
    dst=ADS
    rule=AnimGraph.AnimGraphNode_StateMachine_0.Idle/ADS Hair Anim.AnimStateTransitionNode_0.Transition
    src=Idle
    via=AnimStateTransitionNode_0
   }
   {
    dst=ADS Moving
    rule=AnimGraph.AnimGraphNode_StateMachine_0.Idle/ADS Hair Anim.AnimStateTransitionNode_3.Transition
    src=Idle
    via=AnimStateTransitionNode_3
   }
  ]
 }
 Idle {
  path=AnimGraph.AnimGraphNode_StateMachine_0.Idle/ADS Hair Anim.AnimStateNode_0.Idle
  nodes [Output Animation Pose(Result="Sequence Player (additive) 'ReptilianOcean_Idle_Anim_FaceAcc'"() {sequence=/BRCosmetics/Characters/Player/Female/Medium/Bodies/F_MED_ReptilianOcean_Sleek/Animation/ReptilianOcean_Idle_Anim_FaceAcc})]
 }
 ADS {
  path=AnimGraph.AnimGraphNode_StateMachine_0.Idle/ADS Hair Anim.AnimStateNode_1.ADS
  nodes [
   n1 = "Sequence Player (additive) 'ReptilianOcean_ADS_Anim_FaceAcc'"() {
    groupName=ADS
    method=SyncGroup
    sequence=/BRCosmetics/Characters/Player/Female/Medium/Bodies/F_MED_ReptilianOcean_Sleek/Animation/ReptilianOcean_ADS_Anim_FaceAcc
   }
   n2 = "Sequence Player (additive) 'ReptilianOcean_ADSCrouch__Anim_FaceAcc'"() {
    groupName=ADSCrouch
    method=SyncGroup
    sequence=/BRCosmetics/Characters/Player/Female/Medium/Bodies/F_MED_ReptilianOcean_Sleek/Animation/ReptilianOcean_ADSCrouch__Anim_FaceAcc
   }
   Output Animation Pose(Result=Blend Poses by bool(BlendPose_0=@n2, BlendPose_1=@n1, bActiveValue=Get bIsCrouching()) {blendType=Linear})
  ]
 }
 ADS Moving {
  path=AnimGraph.AnimGraphNode_StateMachine_0.Idle/ADS Hair Anim.AnimStateNode_2.ADS Moving
  nodes [
   n1 = "Sequence Player (additive) 'ReptilianOcean_ADSMoving_Anim_FaceAcc'"() {
    groupName=ADS
    method=SyncGroup
    sequence=/BRCosmetics/Characters/Player/Female/Medium/Bodies/F_MED_ReptilianOcean_Sleek/Animation/ReptilianOcean_ADSMoving_Anim_FaceAcc
   }
   n2 = "Sequence Player (additive) 'ReptilianOcean_ADSCrouchMoving_Anim_FaceAcc'"() {
    groupName=ADSCrouch
    method=SyncGroup
    sequence=/BRCosmetics/Characters/Player/Female/Medium/Bodies/F_MED_ReptilianOcean_Sleek/Animation/ReptilianOcean_ADSCrouchMoving_Anim_FaceAcc
   }
   Output Animation Pose(Result=Blend Poses by bool(BlendPose_0=@n2, BlendPose_1=@n1, bActiveValue=Get bIsCrouching()) {blendType=Linear})
  ]
 }
 Transition {
  path=AnimGraph.AnimGraphNode_StateMachine_0.Idle/ADS Hair Anim.AnimStateTransitionNode_0.Transition
  nodes [Result(bCanEnterTransition=Get bIsTargeting())]
 }
 Transition {
  path=AnimGraph.AnimGraphNode_StateMachine_0.Idle/ADS Hair Anim.AnimStateTransitionNode_1.Transition
  nodes [Result(bCanEnterTransition=NOT Boolean(A=Get bIsTargeting()))]
 }
 Transition {
  path=AnimGraph.AnimGraphNode_StateMachine_0.Idle/ADS Hair Anim.AnimStateTransitionNode_2.Transition
  nodes [Result(bCanEnterTransition=NOT Boolean(A=Get bIsTargeting()))]
 }
 Transition {
  path=AnimGraph.AnimGraphNode_StateMachine_0.Idle/ADS Hair Anim.AnimStateTransitionNode_3.Transition
  nodes [Result(bCanEnterTransition=Get bIsTargeting())]
 }
 Transition {
  path=AnimGraph.AnimGraphNode_StateMachine_0.Idle/ADS Hair Anim.AnimStateTransitionNode_4.Transition
  nodes [Result(bCanEnterTransition=Get bIsMoving2D())]
 }
 Transition {
  path=AnimGraph.AnimGraphNode_StateMachine_0.Idle/ADS Hair Anim.AnimStateTransitionNode_7.Transition
  nodes [Result(bCanEnterTransition=NOT Boolean(A=Get bIsMoving2D()))]
 }
 EventGraph {nodes [
   Event BlueprintUpdateAnimation()
   > n1 = Sequence()
   > n2 = Gravity Override State Function DEPRECATED(CharacterPartAnimInstance=Self-Reference(), GameTimeInSeconds=GetGameTimeInSeconds(), GravityOverrideParams=Get Gravity Override Params())
   > Set Gravity Override Cape(Gravity Override Cape:Vector=@n2) -> (Output_Get:Vector)
   K2Node_VariableGet_customIdle_0 = Get bIsInFrontEnd()
   K2Node_IfThenElse_customIdle_0 = Branch(execute=@n1.then_1, Condition=@K2Node_VariableGet_customIdle_0)
   K2Node_CallFunction_customIdle_0 = GetOwnerFrontEndAnimInstance()
   K2Node_CallFunction_customIdle_1 = GetCurveValue(self=@K2Node_CallFunction_customIdle_0, CurveName=Frontend_Idle_Bypass)
   > K2Node_VariableSet_customIdle_0 = Set Custom Idle is Playing(Custom Idle is Playing:float=@K2Node_CallFunction_customIdle_1) -> (Output_Get:float)
   Event OnCharacterCustomizationLoaded()
   n3 = CheckForBackbling()
   > Set Is Wearing Backbling(Is Wearing Backbling=OR Boolean(A=@n3, B=@n3.bIsBackblingACape))
  ]}
]
variables [
 Control Targets {type=RigidBodyControlTargets, default {targets {}}}
 Custom Idle is Playing {type=number}
 Gravity Override Cape {type=Vector}
 Gravity Override Params {
  type=GravityOverrideParamsStruct
  default {
   onGround_Moving {gravityOverride=<0 0 -150>}
   onGround_Standing {gravityOverride=<0 0 -30>}
   riding_Moving_Targeting {gravityOverride=<0 -200 -800>}
   riding_Standing {gravityOverride=<0 -350 -400>}
   skydive_DivingDown {gravityOverride=<0 0 50>, windAmplitude=<0 0 0>, windFrequency=<0 0 0>}
   skydive_FloatingDown {jointName=pelvis, gravityOverride=<0 0 50>, windAmplitude=<0 0 0>, windFrequency=<0 0 0>, windOffset=<0 0 0>}
   skydive_GoingUp {gravityOverride=<0 -600 -150>}
   skydive_Parachuting {gravityOverride=<0 -500 -800>, windAmplitude=<0 0 0>, windFrequency=<0 0 0>}
   skydive_Surfing {gravityOverride=<0 -500 -980>, windAmplitude=<0 0 0>, windFrequency=<0 0 0>}
  }
 }
 Is Wearing Backbling {type=boolean}
 PlayerPawn {type=FortPlayerPawnAthena}
 rban_alpha {type=number, default=1}
 targeting_cr_alpha {type=number, default=1}
]
